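# Translucent UI material for a circular player avatar with an optional presence indicator for online, away, offline or blocked status.
# The portrait layers the hero texture over a radial-gradient backing, with empty, dimming-overlay and incognito states, or switches to a tinted icon.
# A channel selector dotted with a distance-field texture picks the anti-aliased presence glyph, placed by scale and offset UVs.
# Floored scalar flags and lerps combine these layers into color and opacity passes.
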Material /Game/UI/Foundation/Profile/Materials/M_UI_PlayerAvatar
graph {
 properties {
  bUsedWithStaticMesh=false
  blendMode=BLEND_Translucent
  materialDomain=MD_UI
  pixelDepthOffsetMode=PDOM_Legacy
  usedShadingModels=Default Lit
 }
 nodes [
  NamedRerouteDeclaration(None=Floor(None=ScalarParameter() {group=1 - Presence Icon, parameterName=DisplayPresenceBg, sliderMax=1, sortPriority=33})) {name=Is Showing Presence Bg, nodeColor=<rgba 0 0 0 1>}
  n1 = ScalarParameter() {
   defaultValue=1
   desc="This parameter is changed dynamically at runtime. If the presence icon should NEVER show, use boolean \"Override Display Presence Indicator\"."
   group=1 - Presence Icon
   parameterName=OverrideShowingPresenceIcon
   sliderMax=1
   sortPriority=1
  }
  NamedRerouteDeclaration(None=Floor(None=@n1)) {name=Is Showing Presence Icon, nodeColor=<rgba 0 0 0 1>}
  NamedRerouteDeclaration(None=MaterialFunctionCall() {materialFunction=&GetUserInterfaceUV}.Normalized UV) {name=Normalized UV, nodeColor=<rgba 0 0 0 1>}
  n2 = TextureObjectParameter() {
   group=0 - Portrait
   parameterName=SmallPreview
   sortPriority=0
   texture=/Game/UI/Foundation/Textures/Icons/Heroes/Athena/Soldier/T-Soldier-HID-005-Athena-Commando-M
  }
  NamedRerouteDeclaration(None=TextureSample(UVs=NamedRerouteUsage(), Tex=@n2) {samplerSource=SSM_Clamp_WorldGroupSettings}.RGBA) {name=Texture Pass, nodeColor=<rgba 0 0 0 1>}
  n3 = LinearInterpolate(
   A=NamedRerouteUsage()
   B=NamedRerouteUsage()
   Alpha=ScalarParameter() {group=2 - Icon Portrait, parameterName=Use Icon Color Pass, sortPriority=0}
  )
  n4 = VectorParameter() {group=1 - Presence Icon, parameterName=PresenceBgColor, sortPriority=34, defaultValue=<rgba 0.004 0.0056 0.016 1>}
  n5 = Floor(None=OneMinus(None=NamedRerouteUsage()))
  n6 = NamedRerouteUsage()
  n7 = LinearInterpolate(
   A=NamedRerouteUsage()
   B=LinearInterpolate(A=NamedRerouteDeclaration(None=@n3) {name=Avatar Color, nodeColor=<rgba 0 0 0 1>}, B=@n4.RGB, Alpha=Max(A=@n5, B=@n6))
   Alpha=NamedRerouteUsage()
  )
  n8 = VectorParameter() {group=1 - Presence Icon, parameterName=PresenceIconColor, sortPriority=3, defaultValue=<rgba 1 0 0.75 1>}
  n9 = NamedRerouteUsage()
  n10 = LinearInterpolate(A=Ceil(None=@n9), B=@n9, Alpha=NamedRerouteUsage())
  NamedRerouteDeclaration(None=LinearInterpolate(A=NamedRerouteUsage(), B=LinearInterpolate(A=@n7, B=@n8.RGB, Alpha=@n10) {constAlpha=0}, Alpha=NamedRerouteUsage())) {name=Color Pass, nodeColor=<rgba 0 0 0 1>}
  n11 = NamedRerouteUsage()
  n12 = NamedRerouteUsage()
  n13 = Max(A=Multiply(A=@n11, B=OneMinus(None=@n12)), B=NamedRerouteUsage())
  n14 = Max(A=@n11, B=@n12)
  n15 = NamedRerouteUsage()
  NamedRerouteDeclaration(None=LinearInterpolate(A=NamedRerouteUsage(), B=LinearInterpolate(A=@n13, B=@n14, Alpha=@n15), Alpha=NamedRerouteUsage())) {name=Opacity Pass, nodeColor=<rgba 0 0 0 1>}
  n16 = NamedRerouteUsage()
  n17 = VectorParameter() {
   group=1 - Presence Icon
   parameterName=PresenceIconSelector
   sortPriority=2
   channelNames {a=Is Blocked, b=Is Offline, g=Is Away, r=Is Online}
   defaultValue=<rgba 1 0 0 0>
  }
  n18 = AppendVector(A=@n17, B=@n17)
  n19 = DotProduct(
   A=TextureSample(UVs=@n16, Tex=TextureObject() {samplerType=SAMPLERTYPE_LinearColor, texture=/Game/UI/Foundation/Profile/Textures/T_UI_Icon_Presence}) {samplerSource=SSM_Clamp_WorldGroupSettings}.RGBA
   B=Saturate(None=Floor(None=@n18))
  )
  n20 = Constant() {r=0.375}
  NamedRerouteDeclaration(None=MaterialFunctionCall!(
    UVs=@n16
    Alpha=MaterialFunctionCall(DistanceField=@n19, Threshold=@n20) {materialFunction=&MF_UI_AntiAliasing}.Outer
   ) {materialFunction=/FortUILibrary/MaterialFunctions/Basics/MF_UI_ClipEdgePixel}.Result) {name=Presence Icon Mask, nodeColor=<rgba 0 0 0 1>}
  n21 = NamedRerouteUsage()
  n22 = ScalarParameter() {defaultValue=0.3, group=1 - Presence Icon, parameterName=PresenceIconScale, sortPriority=10}
  n23 = ScalarParameter() {defaultValue=-1, group=1 - Presence Icon, parameterName=PresenceIconOffsetX, sortPriority=11}
  n24 = ScalarParameter() {defaultValue=-1, group=1 - Presence Icon, parameterName=PresenceIconOffsetY, sortPriority=12}
  NamedRerouteDeclaration(None=Add(
    A=MaterialFunctionCall(UVs=@n21, Texture Scale=@n22) {materialFunction=&ScaleUVsByCenter}.UVs
    B=AppendVector(A=@n23, B=@n24)
   )) {name=Presence UV, nodeColor=<rgba 0 0 0 1>}
  n25 = TextureSampleParameter2D(UVs=MaterialFunctionCall(UVs=NamedRerouteUsage(), Texture Scale=Constant() {r=1.5}) {materialFunction=&ScaleUVsByCenter}.UVs) {
   group=2 - Icon Portrait
   parameterName=IconPreview
   samplerType=SAMPLERTYPE_LinearColor
   sortPriority=1
   texture=/FortUILibrary/Textures/Icons/T_UI_Icon_UIKit_TextChatMuted_DF
  }
  NamedRerouteDeclaration(None=MaterialFunctionCall(
    DistanceField=ComponentMask(None=@n25.RGB) {r=true}
    Threshold=ScalarParameter() {defaultValue=0.68, parameterName=Fill Aliasing}
    Smoothing=ScalarParameter() {parameterName=Smoothing, sliderMax=1}
   ) {materialFunction=&MF_UI_AntiAliasing}.Outer) {name=Icon Pass, nodeColor=<rgba 0 0 0 1>}
  NamedRerouteDeclaration(None=TextureSample(
    UVs=NamedRerouteUsage()
    Tex=TextureObject() {samplerType=SAMPLERTYPE_Alpha, texture=/Game/UI/Foundation/Profile/Textures/T_UI_Thumbnail_Portrait_Incognito}
   ) {samplerSource=SSM_Clamp_WorldGroupSettings}.R) {name=Incognito Mask, nodeColor=<rgba 0 0 0 1>}
  NamedRerouteDeclaration(None=MaterialFunctionCall(UVs=NamedRerouteUsage(), Size=Constant() {r=1}) {materialFunction=&MF_UI_Circle}.Fill) {name=Portrait Mask, nodeColor=<rgba 0 0 0 1>}
  n26 = VectorParameter() {group=2 - Icon Portrait, parameterName=Icon Color, sortPriority=2, defaultValue=<rgba 1 0.8817 0.0187 1>}
  NamedRerouteDeclaration(None=LinearInterpolate(A=NamedRerouteUsage(), B=@n26.RGB, Alpha=NamedRerouteUsage())) {name=Icon Color Pass, nodeColor=<rgba 0 0 0 1>}
  n27 = VectorParameter() {group=0 - Portrait, parameterName=BackgroundRGB, defaultValue=<rgba 0.013 0.3613 0.6376 1>}
  n28 = VectorParameter() {group=3 - Empty, parameterName=EmptyBackgroundRGB, sortPriority=0, defaultValue=<rgba 0.013 0.3613 0.6376 1>}
  n29 = NamedRerouteUsage()
  n30 = VectorParameter() {group=0 - Portrait, parameterName=HighlightRGB, defaultValue=<rgba 0.4452 1 1 1>}
  n31 = VectorParameter() {group=3 - Empty, parameterName=EmptyHighlightRGB, sortPriority=1, defaultValue=<rgba 0.4452 1 1 1>}
  n32 = MaterialFunctionCall(UVs=NamedRerouteUsage(), Texture Scale=Constant() {r=1.5}) {materialFunction=&ScaleUVsByCenter}
  n33 = AppendVector(A=Constant(), B=Constant() {r=-0.25})
  n34 = NamedRerouteDeclaration(None=MaterialFunctionCall(UV=Add(A=@n32.UVs, B=@n33)) {materialFunction=&MF_UI_RadialGradient}.Smooth) {name=Radial Gradient, nodeColor=<rgba 0 0 0 1>}
  n35 = Constant() {desc=Radial Intensity, r=0.5}
  NamedRerouteDeclaration(None=LinearInterpolate(
    A=LinearInterpolate(A=@n27.RGB, B=@n28.RGB, Alpha=@n29)
    B=LinearInterpolate(A=@n30.RGB, B=@n31.RGB, Alpha=@n29)
    Alpha=Multiply(A=@n34, B=@n35) {constB=0.5}
   )) {name=Backing Pass, nodeColor=<rgba 0 0 0 1>}
  n36 = NamedRerouteDeclaration(None=Floor(None=ScalarParameter() {group=0 - Portrait, parameterName=IsEmpty, sliderMax=1, sortPriority=1})) {name=Is Empty, nodeColor=<rgba 0 0 0 1>}
  n37 = NamedRerouteUsage()
  n38 = LinearInterpolate(
   A=NamedRerouteUsage()
   B=ComponentMask(None=NamedRerouteUsage()) {b=true, g=true, r=true}
   Alpha=Multiply(A=OneMinus(None=@n36), B=ComponentMask(None=@n37) {a=true})
  )
  n39 = Constant3Vector() {constant=<rgba 0 0.0157 0.0902 1>}
  n40 = Multiply(
   A=NamedRerouteDeclaration(None=ScalarParameter() {group=0 - Portrait, parameterName=OverlayOpacity, sliderMax=1, sortPriority=1}) {name=Overlay Opacity, nodeColor=<rgba 0 0 0 1>}
   B=Constant() {desc=Max Value, r=0.75}
  )
  n41 = VectorParameter() {desc=Incognito Color, parameterName=IncognitoColor, defaultValue=<rgba 0 0.1144 0.3515 1>}
  n42 = VectorParameter() {desc=Inactive Incognito Color, parameterName=InactiveIncognitoColor, defaultValue=<rgba 0 0.0168 0.0887 1>}
  n43 = NamedRerouteUsage()
  n44 = NamedRerouteUsage()
  n45 = NamedRerouteUsage()
  NamedRerouteDeclaration(None=LinearInterpolate(
    A=LinearInterpolate(A=@n38, B=@n39, Alpha=@n40) {constB=0}
    B=LinearInterpolate(A=@n41.RGB, B=@n42.RGB, Alpha=@n43)
    Alpha=Multiply(A=@n44, B=@n45)
   ) {constB=0}) {name=Portrait Color Pass, nodeColor=<rgba 0 0 0 1>}
  NamedRerouteDeclaration(None=MaterialFunctionCall(UVs=NamedRerouteUsage(), Size=Constant() {r=1.1}) {materialFunction=&MF_UI_Circle}.Fill) {name=Presence Mask, nodeColor=<rgba 0 0 0 1>}
  NamedRerouteUsage()
  NamedRerouteUsage()
 ]
}
