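# Base UI kit badge widget showing text over a background image, with an optional status indicator to its left.
# PreConstruct runs InitializeBadge and sets the text; setters store values and update the child widgets.
# RefreshBadge uses size and type enums to select font, padding and background height.
# Padding is set through SlotAsSizeBoxSlot, although Content sits in a GridSlot in this tree.
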WidgetBlueprint /FortUILibrary/Systems/Badge/WBP_UIKit_Badge_Base
graphs [
 EventGraph {nodes [
   Event PreConstruct()
   > InitializeBadge()
   > SetBadgeText!(execute, self, Text:Text=Get BadgeText() -> (BadgeText:Text)) -> (then)
  ]}
 InitializeBadge {nodes [
   InitializeBadge() {is=title K2Node_FunctionEntry}
   > RefreshBadge()
   > SetIndicatorConfiguration!(execute, self, Config:E_UI_BadgeIndicatorType Enum=Get IndicatorConfig() -> (IndicatorConfig:E_UI_BadgeIndicatorType Enum=NewEnumerator3)) -> (then)
   > SetIndicatorType!(
    execute
    self
    IndicatorType:E_UI_StatusIndicator_Type Enum=Get Indicator Type() -> (Indicator Type:E_UI_StatusIndicator_Type Enum=NewEnumerator0)
   ) -> (then)
   > SetIndicatorIcon!(execute, self, IconAsset:Object Reference=Get IndicatorIcon() -> (IndicatorIcon:Object Reference)) -> (then)
  ]}
 RefreshBadge {nodes [
   RefreshBadge() {is=title K2Node_FunctionEntry}
   n1 = Get BadgeSize() -> (BadgeSize:E_UI_BadgeSize Enum=NewEnumerator0)
   n2 = Get TextFont_Compact() -> (TextFont_Compact:Slate Font Info Structure)
   n3 = Get TextFont() -> (TextFont:Slate Font Info Structure)
   n4 = Get NumericFont_Compact() -> (NumericFont_Compact:Slate Font Info Structure)
   n5 = Get NumericFont() -> (NumericFont:Slate Font Info Structure)
   > Set Font!(
    execute
    Font:Slate Font Info Structure=Select!(
     NewEnumerator0:Slate Font Info Structure=Select(NewEnumerator0=@n2, NewEnumerator1=@n3, Index:E_UI_BadgeSize Enum=@n1)
     NewEnumerator1:Slate Font Info Structure=Select(NewEnumerator0=@n4, NewEnumerator1=@n5, Index=@n1)
     Index:E_UI_BadgeType Enum=Get BadgeType() -> (BadgeType:E_UI_BadgeType Enum=NewEnumerator0)
    ) -> (ReturnValue:Slate Font Info Structure)
    self:Text Object Reference=Get Text_Badge() -> (Text_Badge:Common Text Object Reference)
   ) -> (then, Output_Get:Slate Font Info Structure)
   > Set Padding!(
    execute
    Padding:Margin Structure=Select(
     NewEnumerator0:Margin Structure=Get BadgePadding_Compact() -> (BadgePadding_Compact:Margin Structure)
     NewEnumerator1:Margin Structure=Get BadgePadding_Default() -> (BadgePadding_Default:Margin Structure)
     Index=Get BadgeSize() -> (BadgeSize=NewEnumerator0)
    ) -> (ReturnValue:Margin Structure)
    self:Size Box Slot Object Reference=SlotAsSizeBoxSlot!(Widget:Widget Object Reference=Get Content() -> (Content:Grid Panel Object Reference)) -> (ReturnValue:Size Box Slot Object Reference)
   ) -> (then, Output_Get:Margin Structure)
   n6 = Get Background() -> (Background:Image Object Reference)
   > n7 = Set members in Slate Brush!(
    execute:Exec
    StructRef:"Slate Brush Structure (by ref)"=Get Brush!(self:Image Object Reference=@n6) -> (Brush:Slate Brush Structure)
    ImageSize_X:double=16
    ImageSize_Y:double=Select(NewEnumerator0:float=16, NewEnumerator1:float=24, Index=Get BadgeSize() -> (BadgeSize=NewEnumerator0)) -> (ReturnValue:float)
   ) -> (then:Exec, StructOut:"Slate Brush Structure (by ref)") {is=title K2Node_SetFieldsInStruct}
   > SetBrush(self=@n6, InBrush=@n7)
  ]}
 SetBadgeSize {nodes [
   n1 = SetBadgeSize!() -> (then, BadgeSize:E_UI_BadgeSize Enum=NewEnumerator0) {is=title K2Node_FunctionEntry}
   > Set BadgeSize(BadgeSize:E_UI_BadgeSize Enum=@n1) -> (Output_Get:E_UI_BadgeSize Enum=NewEnumerator0)
   > RefreshBadge() {is=title K2Node_CallFunction}
  ]}
 SetBadgeText {nodes [
   n1 = SetBadgeText!() -> (then, Text:Text) {is=title K2Node_FunctionEntry}
   > n2 = Set BadgeText(BadgeText:Text=@n1) -> (Output_Get:Text)
   > Set Text(Text=@n2, self=Get Text_Badge())
  ]}
 SetBadgeType {nodes [
   n1 = SetBadgeType!() -> (then, Type:E_UI_BadgeType Enum=NewEnumerator0) {is=title K2Node_FunctionEntry}
   > Set BadgeType(BadgeType:E_UI_BadgeType Enum=@n1) -> (Output_Get:E_UI_BadgeType Enum=NewEnumerator0)
   > RefreshBadge()
  ]}
 SetIndicatorConfiguration {nodes [
   n1 = SetIndicatorConfiguration!() -> (then, Config:E_UI_BadgeIndicatorType Enum=NewEnumerator0) {is=title K2Node_FunctionEntry}
   > Set IndicatorConfig(IndicatorConfig:E_UI_BadgeIndicatorType Enum=@n1) -> (Output_Get:E_UI_BadgeIndicatorType Enum=NewEnumerator3)
   > SetVisibility(
    self=Get StatusIndicator() -> (StatusIndicator:WBP UIKit Status Indicator Object Reference)
    InVisibility=Select!(
     NewEnumerator3:ESlateVisibility Enum=Collapsed
     NewEnumerator0:ESlateVisibility Enum=HitTestInvisible
     NewEnumerator1:ESlateVisibility Enum=HitTestInvisible
     Index:E_UI_BadgeIndicatorType Enum=Get IndicatorConfig() -> (IndicatorConfig=NewEnumerator3)
    ) -> (ReturnValue:ESlateVisibility Enum)
   )
   > Branch(Condition=Select(
     NewEnumerator3:bool=false
     NewEnumerator0:bool=true
     NewEnumerator1:bool=true
     Index=Get IndicatorConfig() -> (IndicatorConfig=NewEnumerator3)
    ) -> (ReturnValue:bool))
   > SetIndicatorConfiguration!(
    execute
    self:WBP UIKit Status Indicator Base Object Reference=Get StatusIndicator()
    InCount:int=0
    IndicatorConfig:E_UI_StatusIndicator_Configuration Enum=Select(
     NewEnumerator3:E_UI_StatusIndicator_Configuration Enum=NewEnumerator2
     NewEnumerator0:E_UI_StatusIndicator_Configuration Enum=NewEnumerator2
     NewEnumerator1:E_UI_StatusIndicator_Configuration Enum=NewEnumerator0
     Index=Get IndicatorConfig() -> (IndicatorConfig=NewEnumerator3)
    ) -> (ReturnValue:E_UI_StatusIndicator_Configuration Enum)
   ) -> (then) {is=title K2Node_CallFunction}
  ]}
 SetIndicatorIcon {nodes [
   n1 = SetIndicatorIcon!() -> (then, IconAsset:Object Reference) {is=title K2Node_FunctionEntry}
   > n2 = Set IndicatorIcon(IndicatorIcon:Object Reference=@n1) -> (Output_Get:Object Reference)
   > SetIcon!(execute, self:WBP UIKit Status Indicator Base Object Reference=Get StatusIndicator(), IconAsset:Object Reference=@n2) -> (then)
  ]}
 SetIndicatorType {nodes [
   n1 = SetIndicatorType!() -> (then, IndicatorType:E_UI_StatusIndicator_Type Enum=NewEnumerator0) {is=title K2Node_FunctionEntry}
   > n2 = Set Indicator Type(Indicator Type:E_UI_StatusIndicator_Type Enum=@n1) -> (Output_Get:E_UI_StatusIndicator_Type Enum=NewEnumerator0)
   > SetIndicatorType(self:WBP UIKit Status Indicator Base Object Reference=Get StatusIndicator(), IndicatorType=@n2) {is=title K2Node_CallFunction}
  ]}
]
variables [
 BadgePadding_Compact {type=Margin, default=<ltrb 8 0 8 0>}
 BadgePadding_Default {type=Margin, default=<ltrb 12 0 12 0>}
 BadgeSize {type=E_UI_BadgeSize, default=Compact}
 BadgeText {type=string, default=Badge}
 BadgeType {type=E_UI_BadgeType, default=Numeric}
 Indicator Type {type=E_UI_StatusIndicator_Type, default=Positive}
 IndicatorConfig {type=E_UI_BadgeIndicatorType, default=Icon}
 IndicatorIcon {type=Object, default=/FortUILibrary/Textures/MaterialInstances/MI_UIKit_Icon_Check}
 NumericFont {
  type=SlateFontInfo
  default {
   bForceMonospaced=false
   bMaterialIsStencil=false
   fontMaterial=null
   fontObject=&HeadingNow
   letterSpacing=10
   monospacedWidth=1
   size=15.75
   skewAmount=0
   typefaceFontName=86 Bold Italic
   outlineSettings {
    bApplyOutlineToDropShadows=false
    bMiteredCorners=false
    bSeparateFillAlpha=false
    outlineMaterial=null
    outlineSize=0
    outlineColor=<rgba 0 0 0 1>
   }
  }
 }
 NumericFont_Compact {
  type=SlateFontInfo
  default {
   bForceMonospaced=false
   bMaterialIsStencil=false
   fontMaterial=null
   fontObject=&HeadingNow
   letterSpacing=10
   monospacedWidth=1
   size=12.75
   skewAmount=0
   typefaceFontName=75 Medium Italic
   outlineSettings {
    bApplyOutlineToDropShadows=false
    bMiteredCorners=false
    bSeparateFillAlpha=false
    outlineMaterial=null
    outlineSize=0
    outlineColor=<rgba 0 0 0 1>
   }
  }
 }
 TextFont {
  type=SlateFontInfo
  default {
   bForceMonospaced=false
   bMaterialIsStencil=false
   fontMaterial=null
   fontObject=&HeadingNow
   letterSpacing=10
   monospacedWidth=1
   size=15.75
   skewAmount=0
   typefaceFontName=74 Regular
   outlineSettings {
    bApplyOutlineToDropShadows=false
    bMiteredCorners=false
    bSeparateFillAlpha=false
    outlineMaterial=null
    outlineSize=0
    outlineColor=<rgba 0 0 0 1>
   }
  }
 }
 TextFont_Compact {
  type=SlateFontInfo
  default {
   bForceMonospaced=false
   bMaterialIsStencil=false
   fontMaterial=null
   fontObject=&HeadingNow
   letterSpacing=10
   monospacedWidth=1
   size=12.75
   skewAmount=0
   typefaceFontName=74 Regular
   outlineSettings {
    bApplyOutlineToDropShadows=false
    bMiteredCorners=false
    bSeparateFillAlpha=false
    outlineMaterial=null
    outlineSize=0
    outlineColor=<rgba 0 0 0 1>
   }
  }
 }
]
widgets {
 info {inheritedWidgetCount=0, parentClass=&UserWidget, rootWidgetClass=&GridPanel, widgetCount=5}
 widgets [
  Root {widgetClassPath=&GridPanel, widget=Root, properties {columnFill [1], rowFill [1]}}
  Background {
   widgetClassPath=&Image
   bIsVariable=true
   parent=Root
   widget=Background
   properties {
    visibility=SelfHitTestInvisible
    brush {
     bIsDynamicallyLoaded=false
     drawAs=Image
     imageType=FullColor
     mirroring=NoMirror
     resourceName=None
     resourceObject=/FortUILibrary/Systems/Badge/MaterialInstances/MI_UIKit_BadgeBackground
     tiling=NoTile
     imageSize=<16 16>
     margin=<ltrb 0 0 0 0>
     outlineSettings {
      bUseBrushTransparency=false
      roundingType=HalfHeightRadius
      width=0
      color {colorUseRule=UseColor_Specified, specifiedColor=<rgba 0 0 0 0>}
      cornerRadii=<0 0 0 0>
     }
     tintColor {colorUseRule=UseColor_Specified, specifiedColor=<rgba 1 1 1 1>}
     uVRegion {bIsValid=false, max=<0 0>, min=<0 0>}
    }
   }
   slot {class=GridSlot}
  }
  Content {
   widgetClassPath=&GridPanel
   bIsVariable=true
   parent=Root
   widget=Content
   properties {columnFill [0,1], rowFill [1]}
   slot {class=GridSlot, properties {padding=<ltrb 12 0 12 0>}}
  }
  StatusIndicator {
   widgetClassPath=/FortUILibrary/UIKit/StatusIndicator/WBP_UIKit_StatusIndicator
   bIsVariable=true
   parent=Content
   widget=StatusIndicator
   properties {iconSize=Compact, overrideStatusIndicatorVisibility=true}
   slot {class=GridSlot, properties {horizontalAlignment=HAlign_Left, layer=1, verticalAlignment=VAlign_Center, padding=<ltrb 0 0 4 0>}}
  }
  Text_Badge {
   widgetClassPath=&CommonTextBlock
   bIsVariable=true
   parent=Content
   widget=Text_Badge
   properties {
    applyLineHeightToBottomLine=false
    bAutoCollapseWithEmptyText=true
    scrollStyle=&TextScrollStyle-Base_C
    text=Badge
    font {
     bForceMonospaced=false
     bMaterialIsStencil=false
     fontMaterial=null
     fontObject=&BurbankSmall
     letterSpacing=0
     monospacedWidth=1
     size=15.75
     skewAmount=0
     typefaceFontName=Black
     outlineSettings {
      bApplyOutlineToDropShadows=false
      bMiteredCorners=false
      bSeparateFillAlpha=false
      outlineMaterial=null
      outlineSize=0
      outlineColor=<rgba 0 0 0 1>
     }
    }
    shadowOffset=<0 0>
   }
   slot {class=GridSlot, properties {column=1, horizontalAlignment=HAlign_Left, layer=1, verticalAlignment=VAlign_Center}}
  }
 ]
}
