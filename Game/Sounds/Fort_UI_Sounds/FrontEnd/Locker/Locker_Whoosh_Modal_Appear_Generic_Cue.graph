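# Sound cue for a whoosh when a modal appears in the frontend Locker UI.
# A single wave player plays the UI_Locker_QuickSave_Appear sound, lowered in pitch and volume, through the Fort_UI sound class.
# A concurrency override prevents new instances and sets a short retrigger time, limiting rapid repeats.
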SoundCue /Game/Sounds/Fort_UI_Sounds/FrontEnd/Locker/Locker_Whoosh_Modal_Appear_Generic_Cue
root=WavePlayer_0
nodes [{id=WavePlayer_0, properties {soundWaveAssetPtr=&/UI_Locker_QuickSave_Appear}}]
properties {
 bOverrideConcurrency=true
 maxDistance=2097152
 pitchMultiplier=0.8
 soundClassObject=&Fort_UI
 volumeMultiplier=0.2
 concurrencyOverrides {resolutionRule=PreventNew, retriggerTime=0.2}
}
