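# UI material for a banner preview, likely for icon and color selection, compositing a tintable icon over a patterned, colored shape.
# The pattern's offset alpha multiplies the shape texture and a two-color gradient, built from rotated UI UVs and SmoothStep.
# The tinted icon blends over that base using its doubled, clamped alpha.
# A separate BannerShapeColor gradient darkens that color but feeds none of the icon composite.
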Material /Game/UI/Foundation/Banner/BannerIconColorSelection
graph {
 properties {
  bUsedWithStaticMesh=false
  blendMode=BLEND_Translucent
  materialDomain=MD_UI
  pixelDepthOffsetMode=PDOM_Legacy
  shadingModel=MSM_Unlit
  usedShadingModels=Unlit
 }
 nodes [
  n1 = Add(A=TextureSampleParameter2D() {parameterName=BannerPattern, texture=/Game/UI/HomeBase/BannerWidget/HB-Banner-Pattern-2}.A) {constB=0.6}
  n2 = TextureSampleParameter2D() {parameterName=BannerShape, texture=/Game/UI/Foundation/Textures/Buttons/T-Button-Square}
  n3 = VectorParameter() {parameterName=RGBA0, defaultValue=<rgba 0.2279 0.3876 0.7991 1>}
  n4 = VectorParameter() {parameterName=RGBA1, defaultValue=<rgba 0.1339 0.286 0.525 1>}
  n5 = Rotator(
   Coordinate=MaterialFunctionCall() {materialFunction=&GetUserInterfaceUV}.Normalized UV
   Time=ScalarParameter() {group=Linear Gradient Settings, parameterName=ColorRotation}
  )
  n6 = MaterialFunctionCall(
   Alpha=Subtract(A=ComponentMask(None=@n5) {g=true}, B=ScalarParameter() {defaultValue=-0.1, parameterName=ColorOffset})
   Min=ScalarParameter() {group=Linear Gradient Settings, parameterName=RGBAMin}
   Max=ScalarParameter() {defaultValue=1, group=Linear Gradient Settings, parameterName=RGBAMax}
  ) {materialFunction=&SmoothStep}
  n7 = TextureSampleParameter2D() {parameterName=BannerIcon, texture=/Game/UI/HomeBase/BannerWidget/Patterns/HB-Banner-Null}
  LinearInterpolate(
   A=Multiply(A=Multiply(A=@n1, B=@n2.RGB), B=LinearInterpolate(A=@n3.RGB, B=@n4.RGB, Alpha=@n6.Result))
   B=Multiply(A=@n7.RGB, B=VectorParameter() {parameterName=BannerIconColor, defaultValue=<rgba 1 1 1 0>}.RGB)
   Alpha=Clamp(None=Multiply(A=@n7.A) {constB=2})
  )
  n8 = VectorParameter() {parameterName=BannerShapeColor, defaultValue=<rgba 0.1412 0.4488 0.555 1>}
  n9 = ComponentMask(None=Rotator(
    Coordinate=MaterialFunctionCall() {materialFunction=&GetUserInterfaceUV}.Normalized UV
    Time=ScalarParameter() {group=Linear Gradient Settings, parameterName=Color0 Rotation}
   )) {g=true}
  n10 = ScalarParameter() {defaultValue=-0.1, parameterName=Color 0 Offset}
  LinearInterpolate(
   A=@n8.RGB
   B=Multiply(A=@n8.RGB) {constB=0.3}
   Alpha=MaterialFunctionCall(
    Alpha=Subtract(A=@n9, B=@n10)
    Min=ScalarParameter() {defaultValue=0.2, group=Linear Gradient Settings, parameterName=Color0 SS Min}
    Max=ScalarParameter() {defaultValue=1, group=Linear Gradient Settings, parameterName=Color0 SS Max}
   ) {materialFunction=&SmoothStep}.Result
  )
 ]
}
