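# Banner editor list tile in the player profile UI that shows either a banner icon or a banner color item.
# On item set, a macro casts the object to switch layouts, lazy-loading the icon texture or setting two color parameters on a dynamic material.
# A bang wrapper flags unseen items.
# Hovering or selecting marks the item seen, refreshes the bang and broadcasts BannerTileBangUpdated if it was previously unseen.
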WidgetBlueprint /Game/UI/Foundation/Profile/BP_BannerEditorTile
graphs [
 BannerTileBangUpdated {nodes [BannerTileBangUpdated() {is=title K2Node_FunctionEntry}]}
 EventGraph {nodes [
   Event On Hovered()
   > Mark Item As Seen()
   n1 = Event OnListItemObjectSet()
   > Parent: OnListItemObjectSet!(execute, ListItemObject:Object Reference=@n1.ListItemObject) -> (then) {is=title K2Node_CallParentFunction}
   > Update Bang State()
   > n2 = Execute Banner Icon or Color!(
    execute:Exec
    Object:Object Reference=GetListItemObject!(UserObjectListEntry:User Object List Entry Interface) -> (ReturnValue:Object Reference)
   ) -> (BannerIcon:Exec, BannerColor:Exec) {is=title K2Node_MacroInstance}
   SetActiveWidgetIndex(
    execute=@n2.BannerColor
    self=Get CommonWidgetSwitcher_0() -> (CommonWidgetSwitcher_0:Common Widget Switcher Legacy Object Reference)
    Index=1
   )
   > n3 = Cast To FortHomebaseBannerColorItem!(execute, Object=GetListItemObject()) -> (then, CastFailed, AsFort Homebase Banner Color Item:Fort Homebase Banner Color Item Object Reference) {is=title K2Node_DynamicCast}
   > n4 = GetBannerColorPrimaryColor!(execute, BannerColor:Fort Homebase Banner Color Item Object Reference=@n3) -> (then, ReturnValue:Linear Color Structure)
   > n5 = GetBannerColorSecondaryColor!(execute, BannerColor:Fort Homebase Banner Color Item Object Reference=@n3) -> (then, ReturnValue:Linear Color Structure)
   > Is Valid(InputObject=Get ColorMID() -> (ColorMID:Material Instance Dynamic Object Reference))
   > SetVectorParameterValue(self=Get ColorMID(), ParameterName=RGBA0, Value=@n4)
   > SetVectorParameterValue(self=Get ColorMID(), ParameterName=RGBA1, Value=@n5)
   SetActiveWidgetIndex(execute=@n2.BannerIcon, self=Get CommonWidgetSwitcher_0(), Index=0)
   > n6 = Cast To FortHomebaseBannerIconItem!(execute, Object=GetListItemObject()) -> (then, CastFailed, AsFort Homebase Banner Icon Item:Fort Homebase Banner Icon Item Object Reference) {is=title K2Node_DynamicCast}
   > n7 = GetBannerSmallIconAsset!(execute, BannerIcon:Fort Homebase Banner Icon Item Object Reference=@n6, Usage:EBannerUsageContext Enum=BannerIcon) -> (then, ReturnValue:Texture 2D Soft Object Reference)
   > SetBrushFromLazyTexture!(
    execute
    self:Common Lazy Image Object Reference=Get IconImage() -> (IconImage:Fort Lazy Image Object Reference)
    LazyTexture:"Texture 2D Soft Object Reference (by ref)"=@n7
    bMatchSize:bool=false
   ) -> (then)
   Event On Selected()
   > Mark Item As Seen()
   Event Construct()
   > n8 = GetDynamicMaterial(self=Get ColorImage() -> (ColorImage:Image Object Reference))
   > Set ColorMID(ColorMID:Material Instance Dynamic Object Reference=@n8) -> (Output_Get:Material Instance Dynamic Object Reference)
  ]}
 ExecuteBannerIconOrColor {nodes [
   n1 = Inputs!() -> (execute, Object:Object Reference)
   > n2 = Cast To FortHomebaseBannerIconItem(Object=@n1)
   > Outputs!(BannerIcon:Exec, BannerColor:Exec=@n3.then)
   n3 = Cast To FortHomebaseBannerColorItem(execute=@n2.CastFailed, Object=@n1)
  ]}
 Mark Item As Seen {nodes [
   Mark Item As Seen() {is=title K2Node_FunctionEntry}
   > n1 = Cast To FortAccountItem!(execute, Object=GetListItemObject()) -> (then, CastFailed, AsFort Account Item:Fort Account Item Object Reference) {is=title K2Node_DynamicCast}
   > Set bAlreadySeen(bAlreadySeen:bool=HasBeenSeenLocally!(self:Fort Account Item Object Reference=@n1) -> (ReturnValue:bool)) -> (Output_Get:bool)
   > MarkItemAsSeenBP(self:Fort Account Item Object Reference=@n1)
   > Update Bang State()
   > n2 = Branch(Condition=Get bAlreadySeen() -> (bAlreadySeen:bool))
   Call Banner Tile Bang Updated!(execute:Exec=@n2.else, self:BP Banner Editor Tile Object Reference) -> (then:Exec) {is=title K2Node_CallDelegate}
  ]}
 Update Bang State {nodes [
   Update Bang State() {is=title K2Node_FunctionEntry}
   > n1 = Cast To FortAccountItem(Object=GetListItemObject())
   > Set Bang State!(
    execute
    self:Fort Bang Wrapper NUI Object Reference=Get NormalBangWrapper() -> (NormalBangWrapper:Normal Bang Wrapper Object Reference)
    bEnabled:bool=NOT Boolean(A=HasBeenSeenLocally(self=@n1))
    Count:int=0
   ) -> (then) {is=title K2Node_CallFunction}
  ]}
]
variables [
 BannerTileBangUpdated {}
 ColorMID {type=MaterialInstanceDynamic}
]
widgets {
 info {
  inheritedWidgetCount=1
  namedSlotCount=1
  parentClass=/Script/FortniteUI.FortPlayerProfileBannerEditorTile
  rootWidgetClass=&Overlay
  widgetCount=8
 }
 widgets [
  Overlay_1 {widgetClassPath=&Overlay, widget=Overlay_1}
  NormalBangWrapper {
   widgetClassPath=/Game/UI/Foundation/Bangs/NormalBangWrapper
   bIsVariable=true
   parent=Overlay_1
   widget=NormalBangWrapper
   properties {useSmallVersion=true}
   slot {class=OverlaySlot}
  }
  CommonWidgetSwitcher_0 {
   widgetClassPath=&CommonWidgetSwitcherLegacy
   bIsVariable=true
   namedSlotHost=NormalBangWrapper
   widget=CommonWidgetSwitcher_0
   properties {transitionDuration=0}
  }
  Overlay_0 {
   widgetClassPath=&Overlay
   parent=CommonWidgetSwitcher_0
   widget=Overlay_0
   slot {class=WidgetSwitcherSlot, properties {horizontalAlignment=HAlign_Center, verticalAlignment=VAlign_Center}}
  }
  IconImage {
   widgetClassPath=&FortLazyImage
   bIsVariable=true
   parent=Overlay_0
   widget=IconImage
   properties {
    visibility=Visible
    brush {
     bIsDynamicallyLoaded=false
     drawAs=Image
     imageType=NoImage
     mirroring=NoMirror
     resourceName=None
     resourceObject=/Game/UI/Foundation/Textures/Banner/Founders/T-Banners-Icons-Banana-L
     tiling=NoTile
     imageSize=<200 200>
     margin=<ltrb 0 0 0 0>
     outlineSettings {
      bUseBrushTransparency=false
      roundingType=HalfHeightRadius
      width=0
      color {colorUseRule=UseColor_Specified, specifiedColor=<rgba 0 0 0 0>}
      cornerRadii=<0 0 0 0>
     }
     tintColor {colorUseRule=UseColor_Specified, specifiedColor=<rgba 1 1 1 1>}
     uVRegion {bIsValid=false, max=<0 0>, min=<0 0>}
    }
    loadingBackgroundBrush {
     bIsDynamicallyLoaded=false
     drawAs=Image
     imageType=FullColor
     mirroring=NoMirror
     resourceName=None
     resourceObject=/FortUILibrary/Systems/Items/MaterialInstances/MI_UIKit_Item_Loading
     tiling=NoTile
     imageSize=<32 32>
     margin=<ltrb 0 0 0 0>
     outlineSettings {
      bUseBrushTransparency=false
      roundingType=HalfHeightRadius
      width=0
      color {colorUseRule=UseColor_Specified, specifiedColor=<rgba 0 0 0 0>}
      cornerRadii=<0 0 0 0>
     }
     tintColor {colorUseRule=UseColor_Specified, specifiedColor=<rgba 1 1 1 1>}
     uVRegion {bIsValid=false, max=<0 0>, min=<0 0>}
    }
    loadingThrobberBrush {
     bIsDynamicallyLoaded=false
     drawAs=Image
     imageType=FullColor
     mirroring=NoMirror
     resourceName=None
     resourceObject=/Game/Athena/UI/Frontend/Spatial/Art/T_Ui_Empty
     tiling=NoTile
     imageSize=<16 16>
     margin=<ltrb 0 0 0 0>
     outlineSettings {
      bUseBrushTransparency=false
      roundingType=HalfHeightRadius
      width=0
      color {colorUseRule=UseColor_Specified, specifiedColor=<rgba 0 0 0 0>}
      cornerRadii=<0 0 0 0>
     }
     tintColor {colorUseRule=UseColor_Specified, specifiedColor=<rgba 1 1 1 1>}
     uVRegion {bIsValid=false, max=<0 0>, min=<0 0>}
    }
   }
   slot {class=OverlaySlot, properties {horizontalAlignment=HAlign_Fill, verticalAlignment=VAlign_Fill}}
  }
  Image_BannerEditBorder {
   widgetClassPath=&Image
   bIsVariable=true
   parent=Overlay_0
   widget=Image_BannerEditBorder
   properties {brush {
     bIsDynamicallyLoaded=false
     drawAs=Image
     imageType=NoImage
     mirroring=NoMirror
     resourceName=None
     resourceObject=/Game/UI/Foundation/Textures/Banner/Shapes/banner_edit_border
     tiling=NoTile
     imageSize=<128 128>
     margin=<ltrb 0 0 0 0>
     outlineSettings {
      bUseBrushTransparency=false
      roundingType=HalfHeightRadius
      width=0
      color {colorUseRule=UseColor_Specified, specifiedColor=<rgba 0 0 0 0>}
      cornerRadii=<0 0 0 0>
     }
     tintColor {colorUseRule=UseColor_Specified, specifiedColor=<rgba 1 1 1 1>}
     uVRegion {bIsValid=false, max=<0 0>, min=<0 0>}
    }}
   slot {class=OverlaySlot, properties {horizontalAlignment=HAlign_Fill, verticalAlignment=VAlign_Fill}}
  }
  ColorImage {
   widgetClassPath=&Image
   bIsVariable=true
   parent=CommonWidgetSwitcher_0
   widget=ColorImage
   properties {brush {
     bIsDynamicallyLoaded=false
     drawAs=Image
     imageType=NoImage
     mirroring=NoMirror
     resourceName=None
     resourceObject=/Game/UI/Foundation/Banner/BannerIconColorSelection
     tiling=NoTile
     imageSize=<75 75>
     margin=<ltrb 0 0 0 0>
     outlineSettings {
      bUseBrushTransparency=false
      roundingType=HalfHeightRadius
      width=0
      color {colorUseRule=UseColor_Specified, specifiedColor=<rgba 0 0 0 0>}
      cornerRadii=<0 0 0 0>
     }
     tintColor {colorUseRule=UseColor_Specified, specifiedColor=<rgba 1 1 1 1>}
     uVRegion {bIsValid=false, max=<0 0>, min=<0 0>}
    }}
   slot {class=WidgetSwitcherSlot, properties {horizontalAlignment=HAlign_Center, verticalAlignment=VAlign_Center}}
  }
  InputActionWidget {
   widgetClassPath=&CommonActionWidget
   bInherited=true
   bIsVariable=true
   parent=Overlay_1
   widget=InputActionWidget
   properties {renderTransform {angle=0, scale=<1 1>, shear=<0 0>, translation=<0 25>}}
   slot {class=OverlaySlot, properties {horizontalAlignment=HAlign_Center, verticalAlignment=VAlign_Bottom}}
  }
 ]
}
