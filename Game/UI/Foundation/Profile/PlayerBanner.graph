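# PlayerBanner is an account widget that shows a player's banner icon and colors, with an optional season-level surround.
# SetBannerOwner loads the owner's cosmetic loadout and rebinds account-change updates; SetBannerImage takes home base info directly.
# Changed IDs drive BannerImage's dynamic material, lazy-loading small or large icon textures.
# EnableSurround pads the banner and lazy-loads a border-only surround; hover animates the material's ScaleAnimation parameter, reversible or instant.
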WidgetBlueprint /Game/UI/Foundation/Profile/PlayerBanner
animations [Anim_Hovered {
  binding_count=1
  length_seconds=0.15
  object=&self:Anim_Hovered
  sequence {bindings [BannerImage {
     id=#1
     display_name=BannerImage_New
     possessed_class=FortLazyImage
     tracks [{
       type=MovieSceneWidgetMaterialTrack
       sections [{
         type=MovieSceneParameterSection
         end_frame=3
         start_frame=0
         channels [ScaleAnimation {type=MovieSceneScriptingFloatChannel, keys [[0,0],[3,1]]}]
        }]
      }]
    }]}
 }]
graphs [
 ClearBannerState {nodes [
   ClearBannerState() {is=title K2Node_FunctionEntry}
   > Set CurrentBannerIconId(CurrentBannerIconId:Name=None) -> (Output_Get:Name=None)
   > Set CurrentBannerColorId(CurrentBannerColorId:Name=None) -> (Output_Get:Name=None)
  ]}
 EnableSurround {nodes [
   EnableSurround() {is=title K2Node_FunctionEntry}
   > SetupSurround()
   > RefreshSurround!(execute, self, Level:int=0) -> (then)
  ]}
 EventGraph {nodes [
   Account Info Changed!() -> (OutputDelegate, then, NewInfo:Fort Public Account Info Structure) {is=title K2Node_CustomEvent}
   > SetBannerIconAndColorFromProfile()
   > Branch(Condition=Get SurroundEnabled() -> (SurroundEnabled:bool))
   > RefreshSurround(Level=0)
   Event Destruct()
   > UnbindDelegates()
   Event PreConstruct()
   > Update Material Instance()
  ]}
 GetBannerMID {nodes [
   GetBannerMID() {is=title K2Node_FunctionEntry}
   > n1 = GetDynamicMaterial(self=Get BannerImage() -> (BannerImage:Fort Lazy Image Object Reference))
   > Return Node!(execute, BannerMID:Material Instance Dynamic Object Reference=@n1)
  ]}
 PlayHoveredAnimation {nodes [
   n1 = PlayHoveredAnimation!() -> (then, IsHovered:bool=~, Instant:bool=~) {is=title K2Node_FunctionEntry}
   n2 = Get Anim_Hovered() -> (Anim_Hovered:Widget Animation Object Reference)
   > StopAnimation(InAnimation=@n2)
   > n3 = Branch(Condition=@n1)
   n4 = Get Anim_Hovered()
   > PlayAnimation(InAnimation=@n4, StartAtTime=SelectFloat(A=GetEndTime(self=@n4), B=0, bPickA=Get Instant() -> (Instant:bool)))
   PlayAnimation(execute=@n3.else, InAnimation=@n2, StartAtTime=SelectFloat(A=GetEndTime(self=@n2), B=0, bPickA=Get Instant()), PlayMode=Reverse)
  ]}
 RefreshSurround {nodes [
   n1 = RefreshSurround!() -> (then, Level:int=0) {is=title K2Node_FunctionEntry}
   n2 = Break Athena Season Banner Level!(AthenaSeasonBannerLevel:"Athena Season Banner Level Structure (by ref)"=GetSeasonBannerInfo!(self:Fort Account Widget Base Object Reference, Level:int=@n1) -> (ReturnValue:Athena Season Banner Level Structure)) -> (SurroundImage:Texture 2D Soft Object Reference, BannerMaterial:Material Interface Soft Object Reference) {is=title K2Node_BreakStruct}
   > n3 = Branch(Condition=IsAssetNull!(Asset:Soft Object Reference=@n2.BannerMaterial) -> (ReturnValue:bool))
   n4 = Get SurroundImage() -> (SurroundImage:Fort Lazy Image Object Reference)
   > SetVisibility(self=@n4, InVisibility=Hidden)
   SetFromLazyMaterialAndTexture!(
    execute=@n3.else
    self:Fort Lazy Image Object Reference=Get SurroundImage()
    LazyMaterial:"Material Interface Soft Object Reference (by ref)"=@n2.BannerMaterial
    LazyTexture:"Texture 2D Soft Object Reference (by ref)"=@n2
    ParameterName:"Name (by ref)"=Get TextureName() -> (TextureName:Name=None)
   ) -> (then)
   > n5 = GetDynamicMaterial(self=Get SurroundImage())
   > SetScalarParameterValue(self=@n5, ParameterName=BorderOnly, Value=1)
   > SetVisibility(self=@n4, InVisibility=SelfHitTestInvisible)
  ]}
 SetBannerColorId {nodes [
   n1 = SetBannerColorId!() -> (then, InBannerColorId:Name) {is=title K2Node_FunctionEntry}
   > Branch(Condition="Not Equal (Name)"!(A:Name=@n1, B:Name=Get CurrentBannerColorId() -> (CurrentBannerColorId:Name=None)) -> (ReturnValue:bool) {is=title K2Node_CallFunction})
   > n2 = Set CurrentBannerColorId(CurrentBannerColorId=Get InBannerColorId() -> (InBannerColorId:Name=None)) -> (Output_Get=None)
   > n3 = Branch(Condition="Not Equal (Name)"(A=@n2, B=None))
   > n4 = GetBannerColorFromRowName!(execute, BannerColorRowName:"Name (by ref)"=Get CurrentBannerColorId() -> (CurrentBannerColorId=None)) -> (then, ReturnValue:Homebase Banner Color Structure)
   > UpdateBannerColorOnMaterial!(
    execute
    Material:Material Instance Dynamic Object Reference=GetBannerMID!() -> (BannerMID:Material Instance Dynamic Object Reference) {is=title K2Node_CallFunction}
    PrimaryBGColor:Linear Color Structure=GetPrimaryColor!(BannerColor:"Homebase Banner Color Structure (by ref)"=@n4) -> (ReturnValue:Linear Color Structure)
    SecondaryBGColor:Linear Color Structure=GetSecondaryColor!(BannerColor:"Homebase Banner Color Structure (by ref)"=@n4) -> (ReturnValue:Linear Color Structure)
   ) -> (then)
   UpdateBannerColorOnMaterial(execute=@n3.else, Material=GetBannerMID())
  ]}
 SetBannerIconAndColor {nodes [
   n1 = SetBannerIconAndColor!() -> (then, IconId:Name, ColorId:Name) {is=title K2Node_FunctionEntry}
   > SetBannerIconId!(execute, self, InBannerIconId:Name=@n1) -> (then)
   > SetBannerColorId!(execute, self, InBannerColorId:Name=@n1.ColorId) -> (then) {is=title K2Node_CallFunction}
  ]}
 SetBannerIconAndColorFromProfile {nodes [
   SetBannerIconAndColorFromProfile() {is=title K2Node_FunctionEntry}
   n1 = GetCosmeticLoadoutForPlayer!(
    self:Fort Mcp Context Object Reference=Get FortMcpContext!() -> (ReturnValue:Fort Mcp Context Object Reference) {is=title K2Node_GetBlueprintContextEx}
    MemberUniqueId:Unique Net Id Structure=Get OwnerNetId() -> (OwnerNetId:Unique Net Id Structure)
   ) -> (ReturnValue:Cosmetic Loadout Structure)
   > SetBannerIconAndColor!(
    execute
    self
    IconId:Name=String To Name(InString=GetBannerIconIdFromLoadout!(CosmeticLoadout:"Cosmetic Loadout Structure (by ref)"=@n1) -> (ReturnValue:String))
    ColorId:Name=String To Name(InString=GetBannerColorIdFromLoadout!(CosmeticLoadout:"Cosmetic Loadout Structure (by ref)"=@n1) -> (ReturnValue:String))
   ) -> (then) {is=title K2Node_CallFunction}
  ]}
 SetBannerIconId {nodes [
   n1 = SetBannerIconId!() -> (then, InBannerIconId:Name) {is=title K2Node_FunctionEntry}
   > Branch(Condition="Not Equal (Name)"(A=@n1, B=Get CurrentBannerIconId() -> (CurrentBannerIconId:Name=None)))
   > n2 = Set CurrentBannerIconId(CurrentBannerIconId=Get InBannerIconId() -> (InBannerIconId:Name=None)) -> (Output_Get=None)
   > n3 = Branch(Condition="Not Equal (Name)"(A=@n2, B=None))
   > n4 = Branch(Condition=Get bUseLargeTexture() -> (bUseLargeTexture:bool))
   > n5 = GetBannerLargeIconAssetFromRowName!(
    execute
    BannerIconRowName:"Name (by ref)"=Get CurrentBannerIconId() -> (CurrentBannerIconId=None)
    Usage:EBannerUsageContext Enum=BannerIcon
   ) -> (then, ReturnValue:Texture 2D Soft Object Reference)
   n6 = GetBannerSmallIconAssetFromRowName!(
    execute=@n4.else
    BannerIconRowName:"Name (by ref)"=Get CurrentBannerIconId() -> (CurrentBannerIconId=None)
    Usage:EBannerUsageContext Enum=BannerIcon
   ) -> (then, ReturnValue:Texture 2D Soft Object Reference)
   n7 = Merge Execution with Data!(
    Option A:Exec=@n5.then
    Option A Data:Texture 2D Soft Object Reference=@n5
    Option B:Exec=@n6.then
    Option B Data:Texture 2D Soft Object Reference=@n6
   ) -> (Then:Exec, Data:Texture 2D Soft Object Reference) {is=title K2Node_MacroInstance}
   > SetFromLazyMaterialAndTexture(
    self=Get BannerImage()
    LazyMaterial=Make Soft Reference!(Input:Material Instance Dynamic Object Reference=GetBannerMID()) -> (Output:Material Instance Dynamic Soft Object Reference)
    LazyTexture=@n7
    ParameterName=Get BannerIconName() -> (BannerIconName:Name=None)
   )
   UpdateBannerIconOnMaterial!(execute=@n3.else, Material:Material Instance Dynamic Object Reference=GetBannerMID(), Icon:Texture Object Reference) -> (then)
  ]}
 SetBannerImage {nodes [
   n1 = SetBannerImage!() -> (then, Image info:Fort Home Base Info Structure) {is=title K2Node_FunctionEntry}
   n2 = Break Fort Home Base Info!(FortHomeBaseInfo:"Fort Home Base Info Structure (by ref)"=@n1) -> (BannerIconId:String, BannerColorId:String, Name:Text, ValidData:bool=~) {is=title K2Node_BreakStruct}
   > n3 = Branch(Condition=@n2.ValidData)
   > SetBannerIconAndColor(IconId=String To Name(InString=@n2), ColorId=String To Name(InString=@n2.BannerColorId))
   > Return Node!(execute, Success:bool=true)
   Return Node(execute=@n3.else, Success=false)
  ]}
 SetBannerOwner {nodes [
   n1 = SetBannerOwner!() -> (then, OwnerId:Unique Net Id Structure) {is=title K2Node_FunctionEntry}
   > UnbindDelegates()
   > ClearBannerState() {is=title K2Node_CallFunction}
   > n2 = Set OwnerNetId(OwnerNetId:Unique Net Id Structure=@n1) -> (Output_Get:Unique Net Id Structure)
   > Branch(Condition="Is Valid (UniqueIdRepl)"!(InUniqueNetIdRepl:"Unique Net Id Structure (by ref)"=@n2) -> (ReturnValue:bool))
   > SetBannerIconAndColorFromProfile() {is=title K2Node_CallFunction}
   > Bind Event to On Currently Viewed Account Info Changed(self:Fort Mcp Context Object Reference=Get FortMcpContext(), Delegate=Create Event()) {is=title K2Node_AddDelegate}
  ]}
 SetupSurround {nodes [
   SetupSurround() {is=title K2Node_FunctionEntry}
   > Set SurroundEnabled(SurroundEnabled:bool=true) -> (Output_Get:bool)
   > SetVisibility(self=Get SurroundImage(), InVisibility=SelfHitTestInvisible)
   > SetGridSlotPadding!(execute, Widget:Widget Object Reference=Get BannerImage(), Padding:Margin Structure=Make Margin(Left=15, Top=12, Right=15, Bottom=14)) -> (then)
  ]}
 ToggleBannerMask {nodes [
   n1 = ToggleBannerMask!() -> (then, Enable:bool=~) {is=title K2Node_FunctionEntry}
   > SetScalarParameterValue(self=GetBannerMID(), ParameterName=EnableBannerMask, Value=Select(Option 0:double=0, Option 1:double=1, Index=@n1) -> (ReturnValue:double))
  ]}
 UnbindDelegates {nodes [
   UnbindDelegates() {is=title K2Node_FunctionEntry}
   > Unbind Event from On Currently Viewed Account Info Changed(self:Fort Mcp Context Object Reference=Get FortMcpContext(), Delegate=Create Event()) {is=title K2Node_RemoveDelegate}
  ]}
 Update Material Instance {nodes [
   Update Material Instance() {is=title K2Node_FunctionEntry}
   > SetBrushFromMaterial(self=Get BannerImage(), Material=Get BannerMaterialInstance() -> (BannerMaterialInstance:Material Interface Object Reference))
  ]}
]
variables [
 BannerMaterialInstance {type=MaterialInterface, default=/Game/UI/Foundation/Banner/M-Borderless-Banner}
 bUseLargeTexture {type=boolean}
 CurrentBannerColorId {type=string, default=None}
 CurrentBannerIconId {type=string, default=None}
 OwnerNetId {type=UniqueNetIdRepl, default=INVALID}
 SurroundEnabled {type=boolean}
]
widgets {
 info {inheritedWidgetCount=0, parentClass=/Script/FortniteUI.FortAccountWidgetBase, rootWidgetClass=&Overlay, widgetCount=3}
 widgets [
  Overlay_0 {widgetClassPath=&Overlay, widget=Overlay_0}
  BannerImage {
   widgetClassPath=&FortLazyImage
   bIsVariable=true
   parent=Overlay_0
   widget=BannerImage
   properties {
    materialTextureParamName=BannerIcon
    visibility=HitTestInvisible
    brush {
     bIsDynamicallyLoaded=false
     drawAs=Image
     imageType=FullColor
     mirroring=NoMirror
     resourceName=None
     resourceObject=&M-Borderless-Banner
     tiling=NoTile
     imageSize=<32 40>
     margin=<ltrb 0 0 0 0>
     outlineSettings {
      bUseBrushTransparency=false
      roundingType=HalfHeightRadius
      width=0
      color {colorUseRule=UseColor_Specified, specifiedColor=<rgba 0 0 0 0>}
      cornerRadii=<0 0 0 0>
     }
     tintColor {colorUseRule=UseColor_Specified, specifiedColor=<rgba 1 1 1 1>}
     uVRegion {bIsValid=false, max=<0 0>, min=<0 0>}
    }
    loadingBackgroundBrush {
     bIsDynamicallyLoaded=false
     drawAs=Image
     imageType=FullColor
     mirroring=NoMirror
     resourceName=None
     resourceObject=/FortUILibrary/Systems/Items/MaterialInstances/MI_UIKit_Item_Loading
     tiling=NoTile
     imageSize=<32 32>
     margin=<ltrb 0 0 0 0>
     outlineSettings {
      bUseBrushTransparency=false
      roundingType=HalfHeightRadius
      width=0
      color {colorUseRule=UseColor_Specified, specifiedColor=<rgba 0 0 0 0>}
      cornerRadii=<0 0 0 0>
     }
     tintColor {colorUseRule=UseColor_Specified, specifiedColor=<rgba 1 1 1 1>}
     uVRegion {bIsValid=false, max=<0 0>, min=<0 0>}
    }
    loadingThrobberBrush {
     bIsDynamicallyLoaded=false
     drawAs=Image
     imageType=FullColor
     mirroring=NoMirror
     resourceName=None
     resourceObject=/Game/Athena/UI/Frontend/Spatial/Art/T_Ui_Empty
     tiling=NoTile
     imageSize=<16 16>
     margin=<ltrb 0 0 0 0>
     outlineSettings {
      bUseBrushTransparency=false
      roundingType=HalfHeightRadius
      width=0
      color {colorUseRule=UseColor_Specified, specifiedColor=<rgba 0 0 0 0>}
      cornerRadii=<0 0 0 0>
     }
     tintColor {colorUseRule=UseColor_Specified, specifiedColor=<rgba 1 1 1 1>}
     uVRegion {bIsValid=false, max=<0 0>, min=<0 0>}
    }
   }
   slot {class=OverlaySlot, properties {horizontalAlignment=HAlign_Fill, verticalAlignment=VAlign_Fill}}
  }
  SurroundImage {
   widgetClassPath=&FortLazyImage
   bIsVariable=true
   parent=Overlay_0
   widget=SurroundImage
   properties {
    bShowLoading=true
    materialTextureParamName=BannerIcon
    visibility=Collapsed
    brush {
     bIsDynamicallyLoaded=false
     drawAs=Image
     imageType=FullColor
     mirroring=NoMirror
     resourceName=None
     resourceObject=/Game/UI/Foundation/Banner/BR-Banner/M-BR-Banner-Inst-Bronze
     tiling=NoTile
     imageSize=<32 40>
     margin=<ltrb 0 0 0 0>
     outlineSettings {
      bUseBrushTransparency=false
      roundingType=HalfHeightRadius
      width=0
      color {colorUseRule=UseColor_Specified, specifiedColor=<rgba 0 0 0 0>}
      cornerRadii=<0 0 0 0>
     }
     tintColor {colorUseRule=UseColor_Specified, specifiedColor=<rgba 1 1 1 1>}
     uVRegion {bIsValid=false, max=<0 0>, min=<0 0>}
    }
    loadingBackgroundBrush {
     bIsDynamicallyLoaded=false
     drawAs=Image
     imageType=FullColor
     mirroring=NoMirror
     resourceName=None
     resourceObject=/FortUILibrary/UIKit/Throbber/MI_UIKit_Monochrome_Throbber
     tiling=NoTile
     imageSize=<32 32>
     margin=<ltrb 0 0 0 0>
     outlineSettings {
      bUseBrushTransparency=false
      roundingType=HalfHeightRadius
      width=0
      color {colorUseRule=UseColor_Specified, specifiedColor=<rgba 0 0 0 0>}
      cornerRadii=<0 0 0 0>
     }
     tintColor {colorUseRule=UseColor_Specified, specifiedColor=<rgba 1 1 1 1>}
     uVRegion {bIsValid=false, max=<0 0>, min=<0 0>}
    }
    loadingThrobberBrush {
     bIsDynamicallyLoaded=false
     drawAs=Image
     imageType=FullColor
     mirroring=NoMirror
     resourceName=None
     resourceObject=&MI_UIKit_Monochrome_Throbber
     tiling=NoTile
     imageSize=<32 32>
     margin=<ltrb 0 0 0 0>
     outlineSettings {
      bUseBrushTransparency=false
      roundingType=HalfHeightRadius
      width=0
      color {colorUseRule=UseColor_Specified, specifiedColor=<rgba 0 0 0 0>}
      cornerRadii=<0 0 0 0>
     }
     tintColor {colorUseRule=UseColor_Specified, specifiedColor=<rgba 1 1 1 1>}
     uVRegion {bIsValid=false, max=<0 0>, min=<0 0>}
    }
   }
   slot {class=OverlaySlot, properties {horizontalAlignment=HAlign_Fill, verticalAlignment=VAlign_Fill}}
  }
 ]
}
