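# Banner editor screen where the local player picks and applies a profile banner icon and color.
# Icon and Color tabs fill a shared tile view by category; selections update temporary choices and the banner preview, enabling Apply when changed.
# Gamepad input uses a picking-tile state to gate the tiles and category navigation.
# Unseen-item bangs mark tabs and adjacent-category arrows, with tile-triggered refreshes debounced by a timer.
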WidgetBlueprint /Game/UI/Foundation/Profile/WBP_LocalPlayerBannerEditor
graphs [
 Apply Chosen Item {nodes [
   Apply Chosen Item() {is=title K2Node_FunctionEntry}
   > SetBannerIcon!(execute, self:Fort Player Profile Banner Editor Object Reference, Icon:"Name (by ref)"=Get TempChosenIcon() -> (TempChosenIcon:Name=None)) -> (then)
   > SetBannerColor!(execute, self:Fort Player Profile Banner Editor Object Reference, Color:"Name (by ref)"=Get TempChosenColor() -> (TempChosenColor:Name=None)) -> (then)
   > Refresh Apply Button State()
  ]}
 Attempt Leave Picker {nodes [
   Attempt Leave Picker() {is=title K2Node_FunctionEntry}
   > Set Picking Tile State!(execute, self, bPickingTile:bool=false) -> (then) {is=title K2Node_CallFunction}
   > Reset Pickers and Views()
   > Return Node!(execute, Left:bool=false)
  ]}
 Close Banner Editor {nodes [
   Close Banner Editor() {is=title K2Node_FunctionEntry}
   > n1 = Attempt Leave Picker!(execute, self) -> (then, Left:bool) {is=title K2Node_CallFunction}
   > n2 = Branch(Condition=@n1)
   > Return Node!(execute, Passtrough:bool=false)
   Set bViewingIcons(execute=@n2.else, bViewingIcons:bool=true) -> (Output_Get:bool)
   > Reset Icon Category Index()
   > Update Banner Icon Tileview!(execute, self, CurrentCategory:Name=Get ChosenIconCategory() -> (ChosenIconCategory:Name=None)) -> (then)
   > Call Close Banner Editor!(
    execute:Exec
    self:WBP Local Player Banner Editor Object Reference
    IconId:Name=Get ChosenIcon() -> (ChosenIcon:Name=None)
    ColorId:Name=Get ChosenColor() -> (ChosenColor:Name=None)
   ) -> (then:Exec) {is=title K2Node_CallDelegate}
   > Return Node(Passtrough=false)
  ]}
 CloseBannerEditor {nodes [CloseBannerEditor!() -> (then, IconId:Name, ColorId:Name) {is=title K2Node_FunctionEntry}]}
 Decrement Color Category Index {nodes [
   Decrement Color Category Index() {is=title K2Node_FunctionEntry}
   > Set ColorCategoryIndex(ColorCategoryIndex:int=Get Prev Color Category Index!() -> (ReturnValue:int=0) {is=title K2Node_CallFunction}) -> (Output_Get:int=0)
  ]}
 Decrement Icon Category Index {nodes [
   Decrement Icon Category Index() {is=title K2Node_FunctionEntry}
   > Set IconCategoryIndex(IconCategoryIndex:int=Get Prev Icon Category Index!() -> (ReturnValue:int=0) {is=title K2Node_CallFunction}) -> (Output_Get:int=0)
  ]}
 EventGraph {nodes [
   n1 = "On Item Clicked (TileView_Items)"!() -> (OutputDelegate, then, Item:Object Reference) {is=title K2Node_ComponentBoundEvent}
   > n2 = Is Banner Item Changed!(execute, self, Item:Object Reference=@n1.Item) -> (then, bNeedToUpdate:bool)
   n3 = GetIsEnabled!(self:Widget Object Reference=Get ApplyButton() -> (ApplyButton:WBP UIKit Button Regular Object Reference)) -> (ReturnValue:bool)
   n4 = "Equal (Enum)"(
    A=GetCurrentInputType(self=Get CommonInputSubsystem!() -> (ReturnValue:Common Input Object Reference) {is=title K2Node_GetBlueprintContextEx})
    B=Gamepad
   )
   > n5 = Branch(Condition=AND Boolean(A=@n2, B=AND Boolean(A=@n4, B=@n3)))
   Apply Chosen Item(execute=[@n14.then, @n5.then]) {is=title K2Node_CallFunction}
   > CommitChosenIconAndColor(self:Fort Player Profile Banner Editor Object Reference)
   n6 = "On Clicked (Button_Previous)"!() -> (OutputDelegate, then, Button:Common Button Base Object Reference) {is=title K2Node_ComponentBoundEvent}
   Previous Category(execute=[@n6.then, @n25.then])
   "On Clicked (Button_CloseEditorButton)"!() -> (OutputDelegate, then, Button:Common Button Base Object Reference) {is=title K2Node_ComponentBoundEvent}
   > n7 = Attempt Leave Picker()
   > n8 = Branch(Condition=@n7)
   Set bViewingIcons(execute=@n8.else, bViewingIcons=true)
   > Reset Icon Category Index()
   > Update Banner Icon Tileview(CurrentCategory=Get ChosenIconCategory() -> (ChosenIconCategory=None))
   > Call Close Banner Editor(IconId=Get ChosenIcon() -> (ChosenIcon=None), ColorId=Get ChosenColor() -> (ChosenColor=None))
   n9 = "On Item Selection Changed (TileView_Items)"!() -> (OutputDelegate, then, Item:Object Reference, bIsSelected:bool) {is=title K2Node_ComponentBoundEvent}
   > Branch(Condition=@n9.bIsSelected)
   n10 = Get CommonInputSubsystem()
   > n11 = Branch(Condition="Equal (Enum)"(A=GetCurrentInputType(self=@n10), B=Gamepad))
   > n12 = Branch(Condition=Get bPickingTile() -> (bPickingTile:bool))
   Update Temp State On Item Selected!(execute=[@n12.then, @n11.else], self, Item:Object Reference=@n9.Item) -> (then)
   n13 = "On Clicked (Button_Next)"!() -> (OutputDelegate, then, Button:Common Button Base Object Reference) {is=title K2Node_ComponentBoundEvent}
   Next Category(execute=[@n13.then, @n24.then])
   n14 = "On Clicked (ApplyButton)"!() -> (OutputDelegate, then, Button:Common Button Base Object Reference) {is=title K2Node_ComponentBoundEvent}
   n15 = "On Entry Generated (TileView_Items)"!() -> (OutputDelegate, then, Widget:User Widget Object Reference) {is=title K2Node_ComponentBoundEvent}
   > n16 = Cast To BP_BannerEditorTile!(execute, Object=@n15.Widget) -> (then, CastFailed, AsBP Banner Editor Tile:BP Banner Editor Tile Object Reference) {is=title K2Node_DynamicCast}
   > Assign Banner Tile Bang Updated!(execute:Exec, self:BP Banner Editor Tile Object Reference=@n16, Delegate:"Delegate (by ref)"=Create Event()) -> (then:Exec) {is=title K2Node_AssignDelegate}
   n17 = OnTabSelected!() -> (OutputDelegate, then, TabId:Name) {is=title K2Node_CustomEvent}
   > UpdateTabFocus!(execute, self, TabNameID:Name=@n17.TabId) -> (then)
   > n18 = Set bViewingIcons(bViewingIcons="Equal (Name)"(A=ICON, B=@n17.TabId))
   > n19 = Branch(Condition=@n18)
   > Update Banner Icon Tileview(CurrentCategory=Get ChosenIconCategory() -> (ChosenIconCategory=None))
   Update Banner Color Tileview!(execute=@n19.else, self, CurrentCategory:Name=Get ChosenColorCategory() -> (ChosenColorCategory:Name=None)) -> (then)
   BannerTileBangUpdated_Event_0() {is=title K2Node_CustomEvent}
   n20 = Get UpdateBangTimerHandle() -> (UpdateBangTimerHandle:Timer Handle Structure)
   > Branch(Condition=NOT Boolean(A=Is Valid Timer Handle!(Handle:Timer Handle Structure=@n20) -> (ReturnValue:bool)))
   > n21 = Set Timer by Event(Delegate=Create Event(), Time=.5, bLooping=false)
   > Set UpdateBangTimerHandle(UpdateBangTimerHandle:Timer Handle Structure=@n21) -> (Output_Get:Timer Handle Structure)
   Handle Bang State Changed() {is=title K2Node_CustomEvent}
   > Clear and Invalidate Timer by Handle(Handle=Get UpdateBangTimerHandle())
   > Update Bang States()
   Event Construct()
   > Bind Event to On Tab Selected(
    self:Common Tab List Widget Base Object Reference=Get TabCategories() -> (TabCategories:WBP UIKit Tabs Object Reference)
    Delegate=Create Event()
   ) {is=title K2Node_AddDelegate}
   > FoundsPackTextUpdate()
   > Bind Event to On Input Method Changed(self=Get CommonInputSubsystem() {is=title K2Node_GetSubsystem}, Delegate=Create Event())
   n22 = GetCurrentInputType(self=Get CommonInputSubsystem() {is=title K2Node_GetBlueprintContextEx})
   > SetIsEnabled(
    self=Get TileView_Items() -> (TileView_Items:Fort Tile View Object Reference)
    bInIsEnabled=Select(Option 0:bool=true, Option 1:bool=Get bPickingTile(), Index="Equal (Enum)"(A=@n22, B=Gamepad)) -> (ReturnValue:bool)
   )
   > Update Category Navigation Visibility()
   > n23 = Sequence()
   Refresh Apply Button State(execute=@n23.then_1)
   > PlaySound2D(Sound=Get AppearSoundCue() -> (AppearSoundCue:Sound Base Object Reference))
   > SetInputActionsType!(
    execute
    self:WBP UIKit Tabs Base Object Reference=Get TabCategories()
    InputActions:E_UI_TabsInputActionType Enum=NewEnumerator0
    InputActionsWithMouseScroll:bool=false
   ) -> (then)
   Branch(execute=@n23.then_0, Condition=IsMobileGame())
   > SetMinDimensions!(
    execute
    self:Common Button Base Object Reference=[
     Get Button_Next() -> (Button_Next:WBP UIKit Button Generic Object Reference)
     Get Button_Previous() -> (Button_Previous:WBP UIKit Button Generic Object Reference)
    ]
    InMinWidth:int=75
    InMinHeight:int=75
   ) -> (then)
   Event Destruct()
   > Unbind all Events from On Tab Selected!(execute:Exec, self:Common Tab List Widget Base Object Reference=Get TabCategories()) -> (then:Exec) {is=title K2Node_ClearDelegate}
   n24 = Event OnShowNextCategory() {is=title K2Node_Event}
   Event RefreshBannerEditor() {is=title K2Node_Event}
   > SetBannerIconId!(
    execute
    self:Player Banner Object Reference=Get HomebaseBanner() -> (HomebaseBanner:Player Banner Object Reference)
    InBannerIconId:Name=Get ChosenIcon() -> (ChosenIcon=None)
   ) -> (then)
   > SetBannerColorId!(execute, self:Player Banner Object Reference=Get HomebaseBanner(), InBannerColorId:Name=Get ChosenColor() -> (ChosenColor=None)) -> (then)
   n25 = Event OnShowPreviousCategory() {is=title K2Node_Event}
   Event PreConstruct()
   > SetTabButtons()
   n26 = GetActiveTab!(self:Common Tab List Widget Base Object Reference=Get TabList() -> (TabList:Fort Tab List Widget Base Object Reference)) -> (ReturnValue:Name=None)
   Select(Option 0:Name=@n26, Option 1:Name=None, Index:Index Wildcard) -> (ReturnValue:Name)
   Switch on Name!(execute:Exec, Selection:Name=@n26) -> (Default:Exec, Case_0:Exec) {is=title K2Node_SwitchName, pinNames [Case_0]}
  ]}
 FoundsPackTextUpdate {nodes [
   FoundsPackTextUpdate() {is=title K2Node_FunctionEntry}
   n1 = Get FortGlobalUIContext!() -> (ReturnValue:Fort Global UIContext Object Reference) {is=title K2Node_GetBlueprintContextEx}
   > SetVisibility(
    self=Get FoundersNoteText() -> (FoundersNoteText:Common Text Object Reference)
    InVisibility=Select!(
     Campaign:ESlateVisibility Enum=Visible
     Athena:ESlateVisibility Enum=Hidden
     Invalid:ESlateVisibility Enum=Hidden
     Count:ESlateVisibility Enum=Visible
     Creative:ESlateVisibility Enum=Visible
     Index:ESubGame Enum=GetSubGame!(self:Fort Global UIContext Object Reference=@n1) -> (ReturnValue:ESubGame Enum=Campaign)
    ) -> (ReturnValue:ESlateVisibility Enum)
   )
   > Switch on EFortFrontEndFeatureState!(
    execute
    Selection:EFortFrontEndFeatureState Enum=GetFeatureState!(
     self:Fort Front End Context Object Reference=Get FortFrontEndContext!() -> (ReturnValue:Fort Front End Context Object Reference) {is=title K2Node_GetBlueprintContextEx}
     Feature:EFortFrontEndFeature Enum=ShowStore
    ) -> (OutFeatureState:EFortFrontEndFeatureState Enum=Enabled, OutReason:EFortFrontEndFeatureStateReason Enum=Default)
   ) -> (Enabled:Exec, Disabled:Exec, Hidden:Exec, Invalid:Exec) {is=title K2Node_SwitchEnum}
   > SetVisibility(InVisibility=Visible)
  ]}
 Get Next Color Category Index {nodes [
   Get Next Color Category Index() {is=title K2Node_FunctionEntry}
   n1 = Length(TargetArray:Array of Names=Get ColorCategories() -> (ColorCategories:Array of Names))
   n2 = int + int(A=Get ColorCategoryIndex() -> (ColorCategoryIndex:int=0), B=1) -> (ReturnValue=0) {is=title K2Node_CommutativeAssociativeBinaryOperator}
   > Return Node!(
    execute
    ReturnValue:int=Select!(
     Option 0:int=@n2
     Option 1:int=0
     Index:bool="integer >= integer"!(A:int=@n2, B:int=@n1) -> (ReturnValue:bool) {is=title K2Node_CallFunction}
    ) -> (ReturnValue:int)
   )
  ]}
 Get Next Icon Category Index {nodes [
   Get Next Icon Category Index() {is=title K2Node_FunctionEntry}
   n1 = Length(TargetArray=Get IconCategories() -> (IconCategories:Array of Names))
   n2 = int + int(A=Get IconCategoryIndex() -> (IconCategoryIndex:int=0), B=1) -> (ReturnValue=0)
   > Return Node(ReturnValue=Select(Option 0=@n2, Option 1=0, Index="integer >= integer"(A=@n2, B=@n1)))
  ]}
 Get Prev Color Category Index {nodes [
   Get Prev Color Category Index() {is=title K2Node_FunctionEntry}
   n1 = int - int(A=Get ColorCategoryIndex() -> (ColorCategoryIndex=0), B=1) {is=title K2Node_CallFunction}
   n2 = Length(TargetArray=Get ColorCategories())
   > Return Node(ReturnValue=Select(
     Option 0=@n1
     Option 1=int - int(A=@n2, B=1)
     Index="integer < integer"!(A:int=@n1, B:int=0) -> (ReturnValue:bool) {is=title K2Node_CallFunction}
    ))
  ]}
 Get Prev Icon Category Index {nodes [
   Get Prev Icon Category Index() {is=title K2Node_FunctionEntry}
   n1 = int - int(A=Get IconCategoryIndex() -> (IconCategoryIndex=0), B=1)
   n2 = Length(TargetArray=Get IconCategories())
   > Return Node(ReturnValue=Select(Option 0=@n1, Option 1=int - int(A=@n2, B=1), Index="integer < integer"(A=@n1, B=0)))
  ]}
 HandleInputMethodChanged {nodes [
   HandleInputMethodChanged!() -> (then, InputType:ECommonInputType Enum) {is=title K2Node_FunctionEntry}
   > Branch(Condition=Get bIsActiveWidget() -> (bIsActiveWidget:bool))
   > n1 = Branch(Condition="Equal (Enum)"(A=Get InputType() -> (InputType:ECommonInputType Enum), B=Gamepad))
   > Reset Pickers and Views()
   > Set Picking Tile State(bPickingTile=Get bPickingTile())
   Reset for Keyboard and Mouse(execute=@n1.else)
  ]}
 Increment Color Category Index {nodes [
   Increment Color Category Index() {is=title K2Node_FunctionEntry}
   > Set ColorCategoryIndex(ColorCategoryIndex=Get Next Color Category Index!() -> (ReturnValue:int=0) {is=title K2Node_CallFunction}) -> (Output_Get=0)
  ]}
 Increment Icon Category Index {nodes [
   Increment Icon Category Index() {is=title K2Node_FunctionEntry}
   > Set IconCategoryIndex(IconCategoryIndex=Get Next Icon Category Index!() -> (ReturnValue:int=0) {is=title K2Node_CallFunction}) -> (Output_Get=0)
  ]}
 Is Banner Item Changed {nodes [
   n1 = Is Banner Item Changed!() -> (then, Item:Object Reference) {is=title K2Node_FunctionEntry}
   > n2 = Cast To FortHomebaseBannerIconItem!(execute, Object=@n1) -> (then, CastFailed, AsFort Homebase Banner Icon Item:Fort Homebase Banner Icon Item Object Reference) {is=title K2Node_DynamicCast}
   > n3 = GetRowNameForBannerIconItem!(execute, BannerIcon:Fort Homebase Banner Icon Item Object Reference=@n2) -> (then, ReturnValue:Name=None)
   > Return Node!(execute, bNeedToUpdate:bool="Not Equal (Name)"(A=@n3, B=Get TempChosenIcon() -> (TempChosenIcon=None)) -> (ReturnValue=~))
   n4 = Cast To FortHomebaseBannerColorItem!(execute=@n2.CastFailed, Object=@n1) -> (then, CastFailed, AsFort Homebase Banner Color Item:Fort Homebase Banner Color Item Object Reference) {is=title K2Node_DynamicCast}
   > n5 = GetRowNameForBannerColorItem!(execute, BannerColor:Fort Homebase Banner Color Item Object Reference=@n4) -> (then, ReturnValue:Name=None)
   > Return Node(bNeedToUpdate="Not Equal (Name)"(A=@n5, B=Get TempChosenColor() -> (TempChosenColor=None)) -> (ReturnValue=~))
   Return Node(execute=@n4.CastFailed, bNeedToUpdate=false)
  ]}
 IsActiveWidget {nodes [
   IsActiveWidget() {is=title K2Node_FunctionEntry}
   > Return Node!(execute, IsActivated:bool=Get bIsActiveWidget())
  ]}
 Navigate to Chosen Color {nodes [
   Navigate to Chosen Color() {is=title K2Node_FunctionEntry}
   n1 = Get CommonInputSubsystem()
   > Branch(Condition="Equal (Enum)"(A=GetCurrentInputType(self=@n1), B=Gamepad))
   > Set Picking Tile State(bPickingTile=true)
   > n2 = Cast To FortPlayerController(Object=GetOwningPlayer())
   > n3 = GetBannerColorItemFromRowName!(
    execute
    FortPC:Fort Player Controller Object Reference=@n2
    BannerColorRowName:"Name (by ref)"=Get TempChosenColor() -> (TempChosenColor=None)
   ) -> (then, ReturnValue:Fort Homebase Banner Color Item Object Reference)
   > CenterWidget!(execute, ToCenter:Widget Object Reference=Get TileView_Items()) -> (then)
   > Set Selected Item!(execute, self:List View Object Reference=Get TileView_Items(), Item:Object Reference=@n3) -> (then) {is=title K2Node_CallFunction}
   > Navigate To Item!(execute, self:List View Object Reference=Get TileView_Items(), Item:Object Reference=@n3) -> (then)
  ]}
 Navigate to Chosen Icon {nodes [
   Navigate to Chosen Icon() {is=title K2Node_FunctionEntry}
   n1 = Get CommonInputSubsystem()
   > Branch(Condition="Equal (Enum)"(A=GetCurrentInputType(self=@n1), B=Gamepad))
   > Set Picking Tile State(bPickingTile=true)
   > n2 = Cast To FortPlayerController(Object=GetOwningPlayer())
   > n3 = GetBannerIconItemFromRowName!(execute, FortPC:Fort Player Controller Object Reference=@n2, BannerIconRowName:"Name (by ref)"=Get TempChosenIcon() -> (TempChosenIcon=None)) -> (then, ReturnValue:Fort Homebase Banner Icon Item Object Reference)
   > CenterWidget(ToCenter=Get TileView_Items())
   > Set Selected Item(self=Get TileView_Items(), Item=@n3)
   > Navigate To Item(self=Get TileView_Items(), Item=@n3)
  ]}
 Next Category {nodes [
   Next Category() {is=title K2Node_FunctionEntry}
   n1 = GetCurrentInputType(self=Get CommonInputSubsystem())
   > Branch(Condition=OR Boolean(A=Get bPickingTile(), B="Not Equal (Enum)"!(A:ECommonInputType Enum=@n1, B:ECommonInputType Enum=Gamepad) -> (ReturnValue:bool)))
   > n2 = Branch(Condition=Get bViewingIcons() -> (bViewingIcons:bool))
   > Increment Icon Category Index() {is=title K2Node_CallFunction}
   > n3 = Update Banner Icon Tileview(CurrentCategory="Get (a copy)"(Array:Array of Names=Get IconCategories(), Dimension 1=Get IconCategoryIndex() -> (IconCategoryIndex=0)) -> (Output:Name))
   Update Bang States(execute=[@n4.then, @n3.then])
   Increment Color Category Index(execute=@n2.else) {is=title K2Node_CallFunction}
   > n4 = Update Banner Color Tileview(CurrentCategory="Get (a copy)"(Array=Get ColorCategories(), Dimension 1=Get ColorCategoryIndex() -> (ColorCategoryIndex=0)))
  ]}
 OnWidgetActivated {nodes [
   OnWidgetActivated() {is=title K2Node_FunctionEntry}
   > Set bIsActiveWidget(bIsActiveWidget:bool=true) -> (Output_Get:bool)
   > HandleEditorActivated(self:Fort Player Profile Banner Editor Object Reference)
   > Reset Pickers and Views()
   > Update Category Navigation Visibility()
   > SetTabButtons()
   > Update Bang States()
   > Set Picking Tile State(bPickingTile=true)
   > Update Banner Icon Tileview(CurrentCategory=Get ChosenIconCategory() -> (ChosenIconCategory=None))
  ]}
 OnWidgetDeactivated {nodes [
   OnWidgetDeactivated() {is=title K2Node_FunctionEntry}
   > Set bIsActiveWidget(bIsActiveWidget=false)
   > HandleEditorDeactivated(self:Fort Player Profile Banner Editor Object Reference)
   > CommitChosenIconAndColor()
   > ClearBannerState(self:Player Banner Object Reference=Get HomebaseBanner())
  ]}
 Previous Category {nodes [
   Previous Category() {is=title K2Node_FunctionEntry}
   n1 = GetCurrentInputType(self=Get CommonInputSubsystem())
   > Branch(Condition=OR Boolean(A=Get bPickingTile(), B="Not Equal (Enum)"(A=@n1, B=Gamepad)))
   > n2 = Branch(Condition=Get bViewingIcons())
   > Decrement Icon Category Index() {is=title K2Node_CallFunction}
   > n3 = Update Banner Icon Tileview(CurrentCategory="Get (a copy)"(Array=Get IconCategories(), Dimension 1=Get IconCategoryIndex() -> (IconCategoryIndex=0)))
   Update Bang States(execute=[@n4.then, @n3.then])
   Decrement Color Category Index(execute=@n2.else) {is=title K2Node_CallFunction}
   > n4 = Update Banner Color Tileview(CurrentCategory="Get (a copy)"(Array=Get ColorCategories(), Dimension 1=Get ColorCategoryIndex() -> (ColorCategoryIndex=0)))
  ]}
 Refresh Apply Button State {nodes [
   Refresh Apply Button State() {is=title K2Node_FunctionEntry}
   n1 = Get ChosenIcon() -> (ChosenIcon=None)
   n2 = Get TempChosenIcon() -> (TempChosenIcon=None)
   n3 = Get ChosenColor() -> (ChosenColor=None)
   n4 = Get TempChosenColor() -> (TempChosenColor=None)
   > SetIsEnabled(
    self=Get ApplyButton()
    bInIsEnabled=OR Boolean(A="Not Equal (Name)"!(A:Name=@n2, B:Name=@n1) -> (ReturnValue:bool) {is=title K2Node_CallFunction}, B="Not Equal (Name)"(A=@n4, B=@n3))
   )
  ]}
 Reset Color Category Index {nodes [
   Reset Color Category Index() {is=title K2Node_FunctionEntry}
   > Set ColorCategoryIndex(ColorCategoryIndex=Find Item!(TargetArray:Array of Names=Get ColorCategories(), ItemToFind:"Name (by ref)"=Get ChosenColorCategory() -> (ChosenColorCategory=None)) -> (ReturnValue)) -> (Output_Get=0)
  ]}
 Reset Icon Category Index {nodes [
   Reset Icon Category Index() {is=title K2Node_FunctionEntry}
   > Set IconCategoryIndex(IconCategoryIndex=Find Item(TargetArray=Get IconCategories(), ItemToFind=Get ChosenIconCategory() -> (ChosenIconCategory=None))) -> (Output_Get=0)
  ]}
 Reset Pickers and Views {nodes [
   Reset Pickers and Views() {is=title K2Node_FunctionEntry}
   > Set TempChosenIcon(TempChosenIcon:Name=Get ChosenIcon() -> (ChosenIcon=None)) -> (Output_Get:Name=None)
   > Set TempChosenColor(TempChosenColor:Name=Get ChosenColor() -> (ChosenColor=None)) -> (Output_Get:Name=None)
   > Refresh Apply Button State() {is=title K2Node_CallFunction}
   > Set IconCategoryIndex(IconCategoryIndex=0) -> (Output_Get=0)
   > Set ColorCategoryIndex(ColorCategoryIndex=0) -> (Output_Get=0)
   > n1 = Branch(Condition=Get bViewingIcons())
   > Update Banner Icon Tileview(CurrentCategory=Get ChosenIconCategory() -> (ChosenIconCategory=None))
   Update Banner Color Tileview(execute=@n1.else, CurrentCategory=Get ChosenColorCategory() -> (ChosenColorCategory=None))
  ]}
 Reset for Keyboard and Mouse {nodes [
   Reset for Keyboard and Mouse() {is=title K2Node_FunctionEntry}
   > Set Picking Tile State(bPickingTile=true)
   > SetIsEnabled(self=Get TileView_Items(), bInIsEnabled=true)
   > SetIsEnabled(self=[Get Button_Next(), Get Button_Previous()], bInIsEnabled=true)
  ]}
 Select Next Category {nodes [
   Select Next Category() {is=title K2Node_FunctionEntry}
   > Next Category() {is=title K2Node_CallFunction}
   > Return Node!(execute, Passtrough:bool=false)
  ]}
 Select Next Tab {nodes [
   Select Next Tab() {is=title K2Node_FunctionEntry}
   > SelectNextTab!(execute, self:UIKit Tab Group Object Reference=Get TabCategories(), bSuppressClickFeedback:bool=true) -> (then)
   > Return Node(Passtrough=false)
  ]}
 Select Previous Category {nodes [
   Select Previous Category() {is=title K2Node_FunctionEntry}
   > Previous Category() {is=title K2Node_CallFunction}
   > Return Node(Passtrough=false)
  ]}
 Select Previous Tab {nodes [
   Select Previous Tab() {is=title K2Node_FunctionEntry}
   > SelectPreviousTab!(execute, self:UIKit Tab Group Object Reference=Get TabCategories(), bSuppressClickFeedback:bool=true) -> (then)
   > Return Node(Passtrough=false)
  ]}
 Set Current Category {nodes [
   n1 = Set Current Category!() -> (then, CurrentCategory:Name) {is=title K2Node_FunctionEntry}
   > Set CurrentCategory(CurrentCategory:Name=@n1) -> (Output_Get:Name=None)
   > n2 = Branch(Condition=Get bViewingIcons())
   n3 = Get CurrentCategory() -> (CurrentCategory:Name=None)
   > n4 = GetBannerIconCategoryName!(execute, CategoryRowName:"Name (by ref)"=@n3) -> (then, ReturnValue:Text)
   > n5 = Set CategoryDisplayName(CategoryDisplayName:Text=@n4) -> (Output_Get:Text)
   "SetText (Text)"(
    execute=[@n7.then, @n5.then]
    self=Get CategoryText() -> (CategoryText:Common Text Object Reference)
    InText=Get CategoryDisplayName() -> (CategoryDisplayName:Text)
   )
   n6 = GetBannerColorCategoryName!(execute=@n2.else, CategoryRowName:"Name (by ref)"=@n3) -> (then, ReturnValue:Text)
   > n7 = Set CategoryDisplayName(CategoryDisplayName=@n6)
  ]}
 Set Picking Tile State {nodes [
   n1 = Set Picking Tile State!() -> (then, bPickingTile:bool=~) {is=title K2Node_FunctionEntry}
   > Set bPickingTile(bPickingTile:bool=@n1) -> (Output_Get:bool)
   > SetIsEnabled(self=Get TileView_Items(), bInIsEnabled=Get bPickingTile())
   > SetIsEnabled(self=[Get Button_Next(), Get Button_Previous()], bInIsEnabled=Get bPickingTile())
   > Update Category Navigation Visibility()
  ]}
 SetTabButtons {nodes [
   SetTabButtons() {is=title K2Node_FunctionEntry}
   > RemoveAllTabs(self:Common Tab List Widget Base Object Reference=Get TabCategories())
   > n1 = Construct Object from Class(Class=/Script/FortniteUI.UIKitTabGroupButtonViewModel) -> (ReturnValue:UI Kit Tab Group Button View Model Object Reference)
   > Set TabID!(execute, TabID:Name=ICON, self:UI Kit Tab Group Button View Model Object Reference=@n1) -> (then, Output_Get:Name=None)
   > Set Label!(execute, Label:Text="LOCTEXT(\"ICON\")", self:UI Kit Selectable Button View Model Object Reference=@n1) -> (then, Output_Get:Text)
   > Add(
    TargetArray:Array of UI Kit Tab Group Button View Model Object References=Get L_TabButtonMVVM() -> (L_TabButtonMVVM:Array of UI Kit Tab Group Button View Model Object References)
    NewItem:"UI Kit Tab Group Button View Model Object Reference (by ref)"=@n1
   ) {is=title K2Node_CallArrayFunction}
   > n2 = Construct Object from Class(Class=&UIKitTabGroupButtonViewModel)
   > Set TabID(TabID=COLOR, self=@n2) -> (Output_Get=None)
   > Set Label(Label="LOCTEXT(\"COLOR\")", self=@n2)
   > Add(TargetArray=Get L_TabButtonMVVM(), NewItem=@n2)
   > n3 = Construct Object from Class(Class=/Script/FortniteUI.UIKitTabGroupViewModel) -> (ReturnValue:UI Kit Tab Group View Model Object Reference)
   > SetUIKitTabGroupViewModel!(execute, self:WBP UIKit Tabs Object Reference=Get TabCategories(), Viewmodel:UI Kit Tab Group View Model Object Reference=@n3) -> (then)
   > Set Tabs!(
    execute
    Tabs:Array of UI Kit Tab Group Button View Model Object References=Get L_TabButtonMVVM()
    self:UI Kit Tab Group View Model Object Reference=@n3
   ) -> (then, Output_Get:Array of UI Kit Tab Group Button View Model Object References)
   > Set Tabs Data!(
    execute
    self:WBP UIKit Tabs Base Object Reference=Get TabCategories()
    InTabs Data:Array of UI Kit Tab Group Button View Model Object References=Get L_TabButtonMVVM()
   ) -> (then) {is=title K2Node_CallFunction}
   > SelectTabByIndex!(execute, self:UIKit Tab Group Object Reference=Get TabCategories(), Index:int=0, bSuppressClickFeedback:bool=true) -> (then, ReturnValue:bool)
   > n4 = Sequence()
   n5 = Get TabCategories()
   > n6 = Cast To WBP_UIKit_Button_Tab!(
    execute
    Object=GetTabButtonBaseByID!(self:Common Tab List Widget Base Object Reference=@n5, TabNameID:Name=ICON) -> (ReturnValue:Common Button Base Object Reference)
   ) -> (then, CastFailed, AsWBP UIKit Button Tab:WBP UIKit Button Tab Object Reference) {is=title K2Node_DynamicCast}
   > SetIconMaterial!(execute, self:WBP UIKit Button Generic Base Object Reference=@n6, Material Instance:Material Instance Object Reference) -> (then)
   > n7 = Branch(Condition=IsUsingTouch(PlayerController=GetOwningPlayer()))
   > RequestTransitionOnBlocks!(execute, self:UIKit Modular Button Object Reference=@n6, State:EUIKitBlockVisualState Enum=Focused) -> (then)
   RequestTransitionOnBlocks(execute=@n7.else, self=@n6, State=Unfocused)
   n8 = Cast To WBP_UIKit_Button_Tab(execute=@n4.then_1, Object=GetTabButtonBaseByID(self=@n5, TabNameID=COLOR))
   > SetIconMaterial(self=@n8)
   > RequestTransitionOnBlocks(self=@n8, State=Unfocused)
  ]}
 Setup Tileview For Colorswatches {nodes [
   n1 = Setup Tileview For Colorswatches!() -> (then, CategoryRowName:"Name (by ref)") {is=title K2Node_FunctionEntry}
   > n2 = Cast To FortPlayerController(Object=GetOwningPlayer())
   n3 = Get TileView_Items()
   > SetEntryWidth!(execute, self:Tile View Object Reference=@n3, NewWidth:float=64) -> (then)
   > SetEntryHeight!(execute, self:Tile View Object Reference=@n3, NewHeight:float=64) -> (then)
   > n4 = GetBannerColorsForCategory!(execute, FortPC:Fort Player Controller Object Reference=@n2, CategoryRowName:"Name (by ref)"=@n1) -> (then, BannerColors:Array of Fort Account Item Object References)
   > Set List Items!(execute, self:List View Object Reference=Get TileView_Items(), InListItems:Array of Object References=@n4) -> (then) {is=title K2Node_CallFunction}
   > n5 = Cast To FortPlayerController(Object=GetOwningPlayer())
   > n6 = GetBannerColorItemFromRowName(FortPC=@n5, BannerColorRowName=Get TempChosenColor() -> (TempChosenColor=None))
   > n7 = Branch(Condition=Contains Item(TargetArray:Array of Fort Account Item Object References=@n4, ItemToFind:"Fort Account Item Object Reference (by ref)"=@n6))
   > n8 = Set Selected Item(self=Get TileView_Items(), Item=@n6)
   Branch(execute=[@n8.then, @n11.then], Condition=Get bPickingTile())
   n9 = Get TileView_Items()
   > Navigate To Item(self=@n9, Item=Get Selected Item!(self:List View Object Reference=@n9) -> (ReturnValue:Object Reference) {is=title K2Node_CallFunction})
   n10 = Get CommonInputSubsystem()
   Branch(execute=@n7.else, Condition="Equal (Enum)"(A=GetCurrentInputType(self=@n10), B=Gamepad))
   > n11 = SetSelectedIndex!(execute, self:List View Object Reference=Get TileView_Items(), Index:int=0) -> (then)
  ]}
 Setup Tileview For Icons {nodes [
   n1 = Setup Tileview For Icons!() -> (then, CategoryRowName:Name) {is=title K2Node_FunctionEntry}
   > n2 = Cast To FortPlayerController(Object=GetOwningPlayer())
   > SetEntryWidth(self=Get TileView_Items(), NewWidth=104)
   > SetEntryHeight(self=Get TileView_Items(), NewHeight=104)
   > n3 = GetBannerIconsForCategory!(execute, FortPC:Fort Player Controller Object Reference=@n2, CategoryRowName:"Name (by ref)"=@n1) -> (then, BannerIcons:Array of Fort Account Item Object References)
   > Set List Items(self=Get TileView_Items(), InListItems=@n3)
   > n4 = Cast To FortPlayerController(Object=GetOwningPlayer())
   > n5 = GetBannerIconItemFromRowName(FortPC=@n4, BannerIconRowName=Get TempChosenIcon() -> (TempChosenIcon=None))
   > n6 = Branch(Condition=Contains Item(TargetArray=@n3, ItemToFind=@n5))
   > n7 = Set Selected Item(self=Get TileView_Items(), Item=@n5)
   Branch(execute=[@n7.then, @n10.then], Condition=Get bPickingTile())
   n8 = Get TileView_Items()
   > Navigate To Item(self=@n8, Item=Get Selected Item(self=@n8))
   n9 = Get CommonInputSubsystem()
   Branch(execute=@n6.else, Condition="Equal (Enum)"(A=GetCurrentInputType(self=@n9), B=Gamepad))
   > n10 = SetSelectedIndex(self=Get TileView_Items(), Index=0)
  ]}
 Update Bang States {nodes [
   Update Bang States() {is=title K2Node_FunctionEntry}
   > UpdateBangStatesInternal(self:Fort Player Profile Banner Editor Object Reference)
   > Update Icon Button Bang State()
   > Update Color Button Bang State()
   > n1 = Branch(Condition=Get bViewingIcons())
   > Update Left Icon Category Button Bang State()
   > Update Right Icon Category Button Bang State()
   Update Left Color Category Button Bang State(execute=@n1.else)
   > Update Right Color Category Button Bang State()
  ]}
 Update Banner Color Tileview {nodes [
   n1 = Update Banner Color Tileview!() -> (then, CurrentCategory:Name) {is=title K2Node_FunctionEntry}
   > Setup Tileview For Colorswatches!(execute, self, CategoryRowName:"Name (by ref)"=@n1) -> (then) {is=title K2Node_CallFunction}
   > Set Current Category!(execute, self, CurrentCategory:Name=@n1) -> (then) {is=title K2Node_CallFunction}
   > Update Category Navigation Visibility()
  ]}
 Update Banner Icon Tileview {nodes [
   n1 = Update Banner Icon Tileview!() -> (then, CurrentCategory:Name) {is=title K2Node_FunctionEntry}
   > Setup Tileview For Icons!(execute, self, CategoryRowName:Name=@n1) -> (then) {is=title K2Node_CallFunction}
   > Set Current Category(CurrentCategory=@n1)
   > Update Category Navigation Visibility()
  ]}
 Update Category Navigation Visibility {nodes [
   Update Category Navigation Visibility() {is=title K2Node_FunctionEntry}
   n1 = Length(TargetArray=Select(Option 0:Array of Names=Get ColorCategories(), Option 1:Array of Names=Get IconCategories(), Index=Get bViewingIcons()) -> (ReturnValue:Array of Names))
   > Set LocalVisibility(LocalVisibility:ESlateVisibility Enum=Select(
     Option 0:ESlateVisibility Enum=Hidden
     Option 1:ESlateVisibility Enum=SelfHitTestInvisible
     Index="integer > integer"(A=@n1, B=1) {is=title K2Node_CallFunction}
    ) -> (ReturnValue:ESlateVisibility Enum)) -> (Output_Get:ESlateVisibility Enum=Visible)
   > SetVisibility(self=[Get Button_Next(), Get Button_Previous()], InVisibility=Get LocalVisibility() -> (LocalVisibility:ESlateVisibility Enum=Visible))
  ]}
 Update Color Button Bang State {nodes [
   Update Color Button Bang State() {is=title K2Node_FunctionEntry}
   n1 = GetTabButtonBaseByID(self=Get TabCategories(), TabNameID=COLOR)
   > Set Bang State!(
    execute
    self:Fort Bang Wrapper NUI Object Reference=Get StatusIndicator!(self:WBP UIKit Button Generic Base Object Reference=Cast To WBP_UIKit_Button_Tab!(Object=@n1) -> (AsWBP UIKit Button Tab, bSuccess:bool=~)) -> (StatusIndicator:WBP UIKit Status Indicator Object Reference)
    bEnabled:bool=Get bHasUnseenColors() -> (bHasUnseenColors:bool)
    Count:int=0
   ) -> (then) {is=title K2Node_CallFunction}
  ]}
 Update Icon Button Bang State {nodes [
   Update Icon Button Bang State() {is=title K2Node_FunctionEntry}
   n1 = GetTabButtonBaseByID(self=Get TabCategories(), TabNameID=ICON)
   > Set Bang State(
    self=Get StatusIndicator(self=Cast To WBP_UIKit_Button_Tab(Object=@n1) -> (bSuccess=~))
    bEnabled=Get bHasUnseenIcons() -> (bHasUnseenIcons:bool)
    Count=0
   )
  ]}
 Update Left Color Category Button Bang State {nodes [
   Update Left Color Category Button Bang State() {is=title K2Node_FunctionEntry}
   > Set bUnseen(bUnseen:bool=false) -> (Output_Get:bool)
   > n1 = Set PrevIndex(PrevIndex:int=Get Prev Color Category Index() -> (ReturnValue=0) {is=title K2Node_CallFunction}) -> (Output_Get:int=0)
   > n2 = Branch(Condition="Not Equal (Integer)"!(A:int=@n1, B:int=Get ColorCategoryIndex() -> (ColorCategoryIndex=0)) -> (ReturnValue:bool) {is=title K2Node_CallFunction})
   > n3 = Cast To FortPlayerController(Object=GetOwningPlayer())
   > n4 = GetBannerColorsForCategory(FortPC=@n3, CategoryRowName="Get (a copy)"(Array=Get ColorCategories(), Dimension 1=Get PrevIndex() -> (PrevIndex:int=0)))
   > n5 = For Each Loop with Break!(Exec:Exec, Array:Array of Fort Account Item Object References=@n4, Break:Exec=@n6.then) -> (LoopBody:Exec, Array Element:Fort Account Item Object Reference, Array Index:int, Completed:Exec)
   Set Bang State(
    execute=[@n5.Completed, @n3.CastFailed, @n2.else]
    self=Get StatusIndicator(self=Get Button_Previous())
    bEnabled=Get bUnseen() -> (bUnseen:bool)
    Count=0
   )
   Branch(execute=@n5.LoopBody, Condition=NOT Boolean(A=HasBeenSeenLocally!(self:Fort Account Item Object Reference=@n5) -> (ReturnValue:bool)))
   > n6 = Set bUnseen(bUnseen=true)
  ]}
 Update Left Icon Category Button Bang State {nodes [
   Update Left Icon Category Button Bang State() {is=title K2Node_FunctionEntry}
   > Set bUnseen(bUnseen=false)
   > n1 = Set PrevIndex(PrevIndex=Get Prev Icon Category Index() -> (ReturnValue=0) {is=title K2Node_CallFunction}) -> (Output_Get=0)
   > n2 = Branch(Condition="Not Equal (Integer)"(A=@n1, B=Get IconCategoryIndex() -> (IconCategoryIndex=0)))
   > n3 = Cast To FortPlayerController(Object=GetOwningPlayer())
   > n4 = GetBannerIconsForCategory(FortPC=@n3, CategoryRowName="Get (a copy)"(Array=Get IconCategories(), Dimension 1=Get PrevIndex() -> (PrevIndex=0)))
   > n5 = For Each Loop with Break(Array=@n4, Break=@n6.then)
   Set Bang State(execute=[@n5.Completed, @n3.CastFailed, @n2.else], self=Get StatusIndicator(self=Get Button_Previous()), bEnabled=Get bUnseen(), Count=0)
   Branch(execute=@n5.LoopBody, Condition=NOT Boolean(A=HasBeenSeenLocally(self=@n5)))
   > n6 = Set bUnseen(bUnseen=true)
  ]}
 Update Right Color Category Button Bang State {nodes [
   Update Right Color Category Button Bang State() {is=title K2Node_FunctionEntry}
   > Set bUnseen(bUnseen=false)
   > n1 = Set NextIndex(NextIndex:int=Get Next Color Category Index() -> (ReturnValue=0)) -> (Output_Get:int=0)
   > n2 = Branch(Condition="Not Equal (Integer)"(A=@n1, B=Get ColorCategoryIndex() -> (ColorCategoryIndex=0)))
   > n3 = Cast To FortPlayerController(Object=GetOwningPlayer())
   > n4 = GetBannerColorsForCategory(FortPC=@n3, CategoryRowName="Get (a copy)"(Array=Get ColorCategories(), Dimension 1=Get NextIndex() -> (NextIndex:int=0)))
   > n5 = For Each Loop with Break(Array=@n4, Break=@n6.then)
   Set Bang State(execute=[@n5.Completed, @n3.CastFailed, @n2.else], self=Get StatusIndicator(self=Get Button_Next()), bEnabled=Get bUnseen(), Count=0)
   Branch(execute=@n5.LoopBody, Condition=NOT Boolean(A=HasBeenSeenLocally(self=@n5)))
   > n6 = Set bUnseen(bUnseen=true)
  ]}
 Update Right Icon Category Button Bang State {nodes [
   Update Right Icon Category Button Bang State() {is=title K2Node_FunctionEntry}
   > Set bUnseen(bUnseen=false)
   > n1 = Set NextIndex(NextIndex=Get Next Icon Category Index() -> (ReturnValue=0)) -> (Output_Get=0)
   > n2 = Branch(Condition="Not Equal (Integer)"(A=@n1, B=Get IconCategoryIndex() -> (IconCategoryIndex=0)))
   > n3 = Cast To FortPlayerController(Object=GetOwningPlayer())
   > n4 = GetBannerIconsForCategory(FortPC=@n3, CategoryRowName="Get (a copy)"(Array=Get IconCategories(), Dimension 1=Get NextIndex() -> (NextIndex=0)))
   > n5 = For Each Loop with Break(Array=@n4, Break=@n6.then)
   Set Bang State(execute=[@n5.Completed, @n3.CastFailed, @n2.else], self=Get StatusIndicator(self=Get Button_Next()), bEnabled=Get bUnseen(), Count=0)
   Branch(execute=@n5.LoopBody, Condition=NOT Boolean(A=HasBeenSeenLocally(self=@n5)))
   > n6 = Set bUnseen(bUnseen=true)
  ]}
 Update Temp State On Item Selected {nodes [
   n1 = Update Temp State On Item Selected!() -> (then, Item:Object Reference) {is=title K2Node_FunctionEntry}
   > n2 = Cast To FortHomebaseBannerIconItem(Object=@n1)
   > n3 = GetRowNameForBannerIconItem(BannerIcon=@n2) -> (ReturnValue=None)
   > n4 = Set TempChosenIcon(TempChosenIcon=@n3) -> (Output_Get=None)
   > n5 = SetBannerIconId(self=Get HomebaseBanner(), InBannerIconId=@n4)
   Refresh Apply Button State(execute=[@n9.then, @n5.then])
   n6 = Cast To FortHomebaseBannerColorItem(execute=@n2.CastFailed, Object=@n1)
   > n7 = GetRowNameForBannerColorItem(BannerColor=@n6) -> (ReturnValue=None)
   > n8 = Set TempChosenColor(TempChosenColor=@n7) -> (Output_Get=None)
   > n9 = SetBannerColorId(self=Get HomebaseBanner(), InBannerColorId=@n8)
  ]}
 UpdateTabFocus {nodes [
   n1 = UpdateTabFocus!() -> (then, TabNameID:Name) {is=title K2Node_FunctionEntry}
   > n2 = Cast To WBP_UIKit_Button_Tab!(execute:Exec, Object=GetTabButtonBaseByID(self=Get TabCategories(), TabNameID=@n1)) -> (then:Exec, CastFailed:Exec, AsWBP UIKit Button Tab)
   > RequestTransitionOnBlocks(self=@n2, State=Focused)
  ]}
]
variables [
 AppearSoundCue {type=SoundBase, default=/Game/Sounds/Fort_UI_Sounds/FrontEnd/Locker/Locker_Whoosh_Modal_Appear_Generic_Cue}
 bIsActiveWidget {type=boolean}
 bPickingTile {type=boolean}
 ButtonGroup {type=CommonButtonGroupLegacy, default=null}
 bViewingIcons {type=boolean, default=true}
 CloseBannerEditor {}
 ColorCategoryIndex {type=integer}
 CurrentCategory {type=string, default=None}
 Font {
  type=SlateFontInfo
  default {
   bForceMonospaced=false
   bMaterialIsStencil=false
   fontMaterial=null
   fontObject=null
   letterSpacing=0
   monospacedWidth=1
   size=45
   skewAmount=0
   typefaceFontName=None
   outlineSettings {
    bApplyOutlineToDropShadows=false
    bMiteredCorners=false
    bSeparateFillAlpha=false
    outlineMaterial=null
    outlineSize=0
    outlineColor=<rgba 0 0 0 1>
   }
  }
 }
 IconCategoryIndex {type=integer}
 In Brush {
  type=SlateBrush
  default {
   bIsDynamicallyLoaded=false
   drawAs=Image
   imageType=NoImage
   mirroring=NoMirror
   resourceName=None
   resourceObject=null
   tiling=NoTile
   imageSize=<32 32>
   margin=<ltrb 0 0 0 0>
   outlineSettings {
    bUseBrushTransparency=false
    roundingType=HalfHeightRadius
    width=0
    color {colorUseRule=UseColor_Specified, specifiedColor=<rgba 0 0 0 0>}
    cornerRadii=<0 0 0 1>
   }
   tintColor {colorUseRule=UseColor_Specified, specifiedColor=<rgba 1 1 1 0>}
   uVRegion {bIsValid=false, max=<0 0>, min=<0 0>}
  }
 }
 TempChosenColor {type=string, default=None}
 TempChosenIcon {type=string, default=None}
 Text {type=string, default=Icon}
 Text_button {type=string, default=Icon}
 UpdateBangTimerHandle {type=TimerHandle}
]
widgets {
 info {inheritedWidgetCount=2, parentClass=/Script/FortniteUI.FortPlayerProfileBannerEditor, rootWidgetClass=&GridPanel, widgetCount=23}
 widgets [
  Root {widgetClassPath=&GridPanel, widget=Root}
  SizeBox_0 {
   widgetClassPath=&SizeBox
   parent=Root
   widget=SizeBox_0
   properties {bOverride_HeightOverride=true, bOverride_WidthOverride=true, heightOverride=640, widthOverride=1040}
   slot {class=GridSlot, properties {padding=<ltrb 0 40 0 8>}}
  }
  VerticalBox_2 {widgetClassPath=&VerticalBox, parent=SizeBox_0, widget=VerticalBox_2, slot {class=SizeBoxSlot}}
  Header {
   widgetClassPath=&GridPanel
   parent=VerticalBox_2
   widget=Header
   properties {columnFill [1], rowFill [0]}
   slot {class=VerticalBoxSlot}
  }
  CommonTextBlock_0 {
   widgetClassPath=&CommonTextBlock
   parent=Header
   widget=CommonTextBlock_0
   properties {
    scrollStyle=&TextScrollStyle-Base_C
    text=BANNER
    textTransformPolicy=ToUpper
    font {
     bForceMonospaced=false
     bMaterialIsStencil=false
     fontMaterial=null
     fontObject=&HeadingNow
     letterSpacing=1
     monospacedWidth=1
     size=27
     skewAmount=0
     typefaceFontName=86 Bold Italic
     outlineSettings {
      bApplyOutlineToDropShadows=false
      bMiteredCorners=false
      bSeparateFillAlpha=false
      outlineMaterial=null
      outlineSize=0
      outlineColor=<rgba 0 0 0 1>
     }
    }
    shadowOffset=<0 0>
    strikeBrush {
     bIsDynamicallyLoaded=false
     drawAs=NoDrawType
     imageType=NoImage
     mirroring=NoMirror
     resourceName=None
     resourceObject=null
     tiling=NoTile
     imageSize=<32 32>
     margin=<ltrb 0 0 0 0>
     outlineSettings {
      bUseBrushTransparency=false
      roundingType=HalfHeightRadius
      width=0
      color {colorUseRule=UseColor_Specified, specifiedColor=<rgba 0 0 0 0>}
      cornerRadii=<0 0 0 0>
     }
     tintColor {colorUseRule=UseColor_Specified, specifiedColor=<rgba 1 1 1 1>}
     uVRegion {bIsValid=false, max=<0 0>, min=<0 0>}
    }
   }
   slot {class=GridSlot, properties {horizontalAlignment=HAlign_Left, verticalAlignment=VAlign_Top}}
  }
  TabCategories {
   widgetClassPath=/FortUILibrary/UIKit/Tabs/WBP_UIKit_Tabs
   bIsVariable=true
   parent=Header
   widget=TabCategories
   properties {
    initialSelectedTabIndex=0
    isNaturalNavigation=true
    tabSpacing=30
    localTabsViewModels [
     &self:WidgetTree.TabCategories.UIKitTabGroupButtonViewModel_0
     &self:WidgetTree.TabCategories.UIKitTabGroupButtonViewModel_1
    ]
    tabGroupPadding=<ltrb 0 0 0 0>
   }
   slot {class=GridSlot, properties {horizontalAlignment=HAlign_Left, row=1, verticalAlignment=VAlign_Top, padding=<ltrb 0 8 0 0>}}
  }
  Content {
   widgetClassPath=&HorizontalBox
   parent=VerticalBox_2
   widget=Content
   slot {class=VerticalBoxSlot, properties {size {sizeRule=Fill, value=1}}}
  }
  SizeBox_1 {
   widgetClassPath=&SizeBox
   parent=Content
   widget=SizeBox_1
   properties {bOverride_HeightOverride=true, bOverride_WidthOverride=true, heightOverride=222, widthOverride=180}
   slot {class=HorizontalBoxSlot, properties {horizontalAlignment=HAlign_Left, verticalAlignment=VAlign_Top, padding=<ltrb 48 48 48 0>}}
  }
  HomebaseBanner {
   widgetClassPath=&/PlayerBanner
   bIsVariable=true
   parent=SizeBox_1
   widget=HomebaseBanner
   properties {bUseLargeTexture=true}
   slot {class=SizeBoxSlot}
  }
  VerticalBox_3 {
   widgetClassPath=&VerticalBox
   parent=Content
   widget=VerticalBox_3
   slot {class=HorizontalBoxSlot, properties {size {sizeRule=Fill, value=1}}}
  }
  Rotator {
   widgetClassPath=&Border
   parent=VerticalBox_3
   widget=Rotator
   properties {
    background {
     bIsDynamicallyLoaded=false
     drawAs=Box
     imageType=NoImage
     mirroring=NoMirror
     resourceName=None
     resourceObject=/FortUILibrary/Foundation/Art_ToolKit/Tools/M_UI_Circle_White
     tiling=NoTile
     imageSize=<32 32>
     margin=<ltrb 0.5 0.5 0.5 0.5>
     outlineSettings {
      bUseBrushTransparency=false
      roundingType=HalfHeightRadius
      width=0
      color {colorUseRule=UseColor_Specified, specifiedColor=<rgba 0 0 0 0>}
      cornerRadii=<0 0 0 0>
     }
     tintColor {colorUseRule=UseColor_Specified, specifiedColor=<rgba 1 1 1 1>}
     uVRegion {bIsValid=false, max=<0 0>, min=<0 0>}
    }
    brushColor=<rgba 0 0 0 0.3>
    padding=<ltrb 5 5 5 5>
   }
   slot {class=VerticalBoxSlot, properties {padding=<ltrb 80 20 70 0>}}
  }
  HorizontalBox_2 {
   widgetClassPath=&HorizontalBox
   parent=Rotator
   widget=HorizontalBox_2
   slot {class=BorderSlot, properties {padding=<ltrb 5 5 5 5>}}
  }
  Button_Previous {
   widgetClassPath=&WBP_UIKit_Button_Generic
   bIsVariable=true
   parent=HorizontalBox_2
   widget=Button_Previous
   properties {
    bIsFocusable=false
    backgroundMaterial=/FortUILibrary/Systems/Backplate/MaterialInstances/MI_UIKit_Backplate
    icon=/FortUILibrary/Textures/Icons/ToBeDeprecated/T_UI_Icon_Global_Chevron_Left
    iconMaterial=/FortUILibrary/UIKit/Buttons/MaterialInstances/GenericButtons/MI_GenericIconDeepBlue
    minHeight=50
    minWidth=50
    outlineMaterial=/FortUILibrary/UIKit/Buttons/MaterialInstances/TileOutline/MI_UIKit_Outline
    shouldTextImpactLayout=false
    statusIndicatorVerticalAlignment=VAlign_Center
    text=""
    useStatusIndicator=true
    useTextBlock=false
    buttonPadding=<ltrb 8 0 8 0>
    iconSize=<30 30>
    modularBlocks [null,null]
    promptNudge=<-50 0>
    statusIndicatorPadding=<ltrb 0 0 4 0>
    triggeringInputAction {dataTable=&InputActionDataTable, rowName=GenericLeftTrigger}
   }
   slot {class=HorizontalBoxSlot}
  }
  HorizontalBox_0 {
   widgetClassPath=&HorizontalBox
   parent=HorizontalBox_2
   widget=HorizontalBox_0
   slot {
    class=HorizontalBoxSlot
    properties {horizontalAlignment=HAlign_Center, verticalAlignment=VAlign_Center, padding=<ltrb 50 0 50 0>, size {sizeRule=Fill, value=1}}
   }
  }
  CategoryStatusIndicator {
   widgetClassPath=/FortUILibrary/UIKit/StatusIndicator/WBP_UIKit_StatusIndicator
   bIsVariable=true
   parent=HorizontalBox_0
   widget=CategoryStatusIndicator
   properties {type=Attention, configuration=Dot}
   slot {class=HorizontalBoxSlot, properties {padding=<ltrb 0 0 10 0>}}
  }
  CategoryText {
   widgetClassPath=&CommonTextBlock
   bIsVariable=true
   parent=HorizontalBox_0
   widget=CategoryText
   properties {
    text=Category
    font {
     bForceMonospaced=false
     bMaterialIsStencil=false
     fontMaterial=null
     fontObject=&HeadingNow
     letterSpacing=0
     monospacedWidth=1
     size=18
     skewAmount=0
     typefaceFontName=75 Medium
     outlineSettings {
      bApplyOutlineToDropShadows=false
      bMiteredCorners=false
      bSeparateFillAlpha=false
      outlineMaterial=null
      outlineSize=0
      outlineColor=<rgba 0 0 0 1>
     }
    }
    shadowOffset=<0 0>
   }
   slot {class=HorizontalBoxSlot}
  }
  Button_Next {
   widgetClassPath=&WBP_UIKit_Button_Generic
   bIsVariable=true
   parent=HorizontalBox_2
   widget=Button_Next
   properties {
    bIsFocusable=false
    backgroundMaterial=&MI_UIKit_Backplate
    holdData=null
    icon=/FortUILibrary/Textures/Icons/ToBeDeprecated/T_UI_Icon_Global_Chevron_Right
    iconMaterial=&MI_GenericIconDeepBlue
    minHeight=50
    minWidth=50
    outlineMaterial=&MI_UIKit_Outline
    promptHorizontalAlignment=HAlign_Right
    shouldTextImpactLayout=false
    statusIndicatorHorizontalAlignment=HAlign_Left
    statusIndicatorVerticalAlignment=VAlign_Center
    text=""
    useStatusIndicator=true
    useTextBlock=false
    buttonPadding=<ltrb 8 0 8 0>
    iconSize=<30 30>
    promptNudge=<50 0>
    statusIndicatorPadding=<ltrb 4 0 0 0>
    triggeringInputAction {dataTable=&InputActionDataTable, rowName=GenericRightTrigger}
   }
   slot {class=HorizontalBoxSlot}
  }
  TileView_Items {
   widgetClassPath=/Script/FortniteUI.FortTileView
   bIsVariable=true
   parent=VerticalBox_3
   widget=TileView_Items
   properties {
    bConsumeTouchInputWhenScrolling=false
    bEnableMobileScrollBarThicknessOverride=true
    bIsEnabled=false
    entryHeight=0
    entryWidgetClass=&/BP_BannerEditorTile.BP_BannerEditorTile_C
    entryWidth=0
    horizontalEntrySpacing=10
    mobileScrollBarThicknessOverride=24
    navigation=&self:WidgetTree.TileView_Items.WidgetNavigation_0
    tileAlignment=LeftAligned
    verticalEntrySpacing=10
    scrollBarStyle {
     thickness=10
     draggedThumbImage {
      bIsDynamicallyLoaded=false
      drawAs=Box
      imageType=FullColor
      mirroring=NoMirror
      resourceName=../../../Engine/Content/Slate/Common/Scrollbar_Thumb.png
      resourceObject=&M_UI_Circle_White
      tiling=NoTile
      imageSize=<10 10>
      margin=<ltrb 0.5 0.5 0.5 0.5>
      outlineSettings {
       bUseBrushTransparency=false
       roundingType=HalfHeightRadius
       width=0
       color {colorUseRule=UseColor_Specified, specifiedColor=<rgba 0 0 0 0>}
       cornerRadii=<0 0 0 0>
      }
      tintColor {colorUseRule=UseColor_Specified, specifiedColor=<rgba 1 1 1 1>}
      uVRegion {bIsValid=false, max=<0 0>, min=<0 0>}
     }
     horizontalBackgroundImage {
      bIsDynamicallyLoaded=false
      drawAs=NoDrawType
      imageType=NoImage
      mirroring=NoMirror
      resourceName=None
      resourceObject=null
      tiling=NoTile
      imageSize=<0 0>
      margin=<ltrb 0 0 0 0>
      outlineSettings {
       bUseBrushTransparency=false
       roundingType=HalfHeightRadius
       width=0
       color {colorUseRule=UseColor_Specified, specifiedColor=<rgba 0 0 0 0>}
       cornerRadii=<0 0 0 0>
      }
      tintColor {colorUseRule=UseColor_Specified, specifiedColor=<rgba 1 1 1 1>}
      uVRegion {bIsValid=false, max=<0 0>, min=<0 0>}
     }
     horizontalBottomSlotImage {
      bIsDynamicallyLoaded=false
      drawAs=Image
      imageType=FullColor
      mirroring=NoMirror
      resourceName=../../../Engine/Content/Slate/Common/Scrollbar_Background_Horizontal.png
      resourceObject=null
      tiling=NoTile
      imageSize=<8 8>
      margin=<ltrb 0 0 0 0>
      outlineSettings {
       bUseBrushTransparency=false
       roundingType=HalfHeightRadius
       width=0
       color {colorUseRule=UseColor_Specified, specifiedColor=<rgba 0 0 0 0>}
       cornerRadii=<0 0 0 0>
      }
      tintColor {colorUseRule=UseColor_Specified, specifiedColor=<rgba 1 1 1 1>}
      uVRegion {bIsValid=false, max=<0 0>, min=<0 0>}
     }
     horizontalTopSlotImage {
      bIsDynamicallyLoaded=false
      drawAs=Image
      imageType=FullColor
      mirroring=NoMirror
      resourceName=../../../Engine/Content/Slate/Common/Scrollbar_Background_Horizontal.png
      resourceObject=null
      tiling=NoTile
      imageSize=<8 8>
      margin=<ltrb 0 0 0 0>
      outlineSettings {
       bUseBrushTransparency=false
       roundingType=HalfHeightRadius
       width=0
       color {colorUseRule=UseColor_Specified, specifiedColor=<rgba 0 0 0 0>}
       cornerRadii=<0 0 0 0>
      }
      tintColor {colorUseRule=UseColor_Specified, specifiedColor=<rgba 1 1 1 1>}
      uVRegion {bIsValid=false, max=<0 0>, min=<0 0>}
     }
     hoveredThumbImage {
      bIsDynamicallyLoaded=false
      drawAs=Box
      imageType=FullColor
      mirroring=NoMirror
      resourceName=../../../Engine/Content/Slate/Common/Scrollbar_Thumb.png
      resourceObject=&M_UI_Circle_White
      tiling=NoTile
      imageSize=<10 10>
      margin=<ltrb 0.5 0.5 0.5 0.5>
      outlineSettings {
       bUseBrushTransparency=false
       roundingType=HalfHeightRadius
       width=0
       color {colorUseRule=UseColor_Specified, specifiedColor=<rgba 0 0 0 0>}
       cornerRadii=<0 0 0 0>
      }
      tintColor {colorUseRule=UseColor_Specified, specifiedColor=<rgba 1 1 1 1>}
      uVRegion {bIsValid=false, max=<0 0>, min=<0 0>}
     }
     normalThumbImage {
      bIsDynamicallyLoaded=false
      drawAs=Box
      imageType=FullColor
      mirroring=NoMirror
      resourceName=../../../Engine/Content/Slate/Common/Scrollbar_Thumb.png
      resourceObject=&M_UI_Circle_White
      tiling=NoTile
      imageSize=<10 10>
      margin=<ltrb 0.5 0.5 0.5 0.5>
      outlineSettings {
       bUseBrushTransparency=false
       roundingType=HalfHeightRadius
       width=0
       color {colorUseRule=UseColor_Specified, specifiedColor=<rgba 0 0 0 0>}
       cornerRadii=<0 0 0 0>
      }
      tintColor {colorUseRule=UseColor_Specified, specifiedColor=<rgba 1 1 1 1>}
      uVRegion {bIsValid=false, max=<0 0>, min=<0 0>}
     }
     verticalBackgroundImage {
      bIsDynamicallyLoaded=false
      drawAs=Box
      imageType=NoImage
      mirroring=NoMirror
      resourceName=None
      resourceObject=&M_UI_Circle_White
      tiling=NoTile
      imageSize=<10 10>
      margin=<ltrb 0.5 0.5 0.5 0.5>
      outlineSettings {
       bUseBrushTransparency=false
       roundingType=HalfHeightRadius
       width=0
       color {colorUseRule=UseColor_Specified, specifiedColor=<rgba 0 0 0 0>}
       cornerRadii=<0 0 0 0>
      }
      tintColor {colorUseRule=UseColor_Specified, specifiedColor=<rgba 1 1 1 0.2>}
      uVRegion {bIsValid=false, max=<0 0>, min=<0 0>}
     }
     verticalBottomSlotImage {
      bIsDynamicallyLoaded=false
      drawAs=NoDrawType
      imageType=FullColor
      mirroring=NoMirror
      resourceName=../../../Engine/Content/Slate/Common/Scrollbar_Background_Vertical.png
      resourceObject=null
      tiling=NoTile
      imageSize=<8 8>
      margin=<ltrb 0 0 0 0>
      outlineSettings {
       bUseBrushTransparency=false
       roundingType=HalfHeightRadius
       width=0
       color {colorUseRule=UseColor_Specified, specifiedColor=<rgba 0 0 0 0>}
       cornerRadii=<0 0 0 0>
      }
      tintColor {colorUseRule=UseColor_Specified, specifiedColor=<rgba 1 1 1 1>}
      uVRegion {bIsValid=false, max=<0 0>, min=<0 0>}
     }
     verticalTopSlotImage {
      bIsDynamicallyLoaded=false
      drawAs=NoDrawType
      imageType=FullColor
      mirroring=NoMirror
      resourceName=../../../Engine/Content/Slate/Common/Scrollbar_Background_Vertical.png
      resourceObject=null
      tiling=NoTile
      imageSize=<8 8>
      margin=<ltrb 0 0 0 0>
      outlineSettings {
       bUseBrushTransparency=false
       roundingType=HalfHeightRadius
       width=0
       color {colorUseRule=UseColor_Specified, specifiedColor=<rgba 0 0 0 0>}
       cornerRadii=<0 0 0 0>
      }
      tintColor {colorUseRule=UseColor_Specified, specifiedColor=<rgba 1 1 1 1>}
      uVRegion {bIsValid=false, max=<0 0>, min=<0 0>}
     }
    }
    widgetStyle {backgroundBrush {
      bIsDynamicallyLoaded=false
      drawAs=Box
      imageType=NoImage
      mirroring=NoMirror
      resourceName=None
      resourceObject=&M_UI_Circle_White
      tiling=NoTile
      imageSize=<10 10>
      margin=<ltrb 0.5 0.5 0.5 0.5>
      outlineSettings {
       bUseBrushTransparency=false
       roundingType=HalfHeightRadius
       width=0
       color {colorUseRule=UseColor_Specified, specifiedColor=<rgba 0 0 0 0>}
       cornerRadii=<0 0 0 0>
      }
      tintColor {colorUseRule=UseColor_Specified, specifiedColor=<rgba 1 1 1 0.2>}
      uVRegion {bIsValid=false, max=<0 0>, min=<0 0>}
     }}
   }
   slot {class=VerticalBoxSlot, properties {padding=<ltrb 30 10 0 0>, size {sizeRule=Fill, value=1}}}
  }
  ButtonsSlot {
   widgetClassPath=&HorizontalBox
   parent=VerticalBox_2
   widget=ButtonsSlot
   slot {class=VerticalBoxSlot, properties {verticalAlignment=VAlign_Bottom, padding=<ltrb 0 35 0 20>}}
  }
  ApplyButton {
   widgetClassPath=&WBP_UIKit_Button_Regular
   bIsVariable=true
   parent=ButtonsSlot
   widget=ApplyButton
   properties {
    buttonSize=Small
    minHeight=56
    text=Apply
    hoveredSlateSoundOverride {resourceObject=/Game/Sounds/Fort_UI_Sounds/FrontEnd/Locker/Locker_Button_Hover_Default_Cue}
    lockedHoveredSlateSoundOverride {resourceObject=&Locker_Button_Hover_Default_Cue}
    lockedPressedSlateSoundOverride {resourceObject=/Game/Sounds/Fort_UI_Sounds/FrontEnd/CareerPage/Fortnite_UI_FlagScreenButton_Cue}
    pressedSlateSoundOverride {resourceObject=&Fortnite_UI_FlagScreenButton_Cue}
    selectedHoveredSlateSoundOverride {resourceObject=&Locker_Button_Hover_Default_Cue}
    selectedPressedSlateSoundOverride {resourceObject=&Fortnite_UI_FlagScreenButton_Cue}
    triggeringInputAction {dataTable=&InputActionDataTable, rowName=Save}
   }
   slot {class=HorizontalBoxSlot, properties {padding=<ltrb 12 0 12 0>, size {sizeRule=Fill, value=1}}}
  }
  Button_CloseEditorButton {
   widgetClassPath=&WBP_UIKit_Button_Regular
   bInherited=true
   bIsVariable=true
   parent=ButtonsSlot
   widget=Button_CloseEditorButton
   properties {
    bHideInputActionWithKeyboard=true
    buttonSize=Small
    minHeight=56
    text=BACK
    hoveredSlateSoundOverride {resourceObject=&Locker_Button_Hover_Default_Cue}
    lockedHoveredSlateSoundOverride {resourceObject=&Locker_Button_Hover_Default_Cue}
    lockedPressedSlateSoundOverride {resourceObject=/Game/Sounds/Fort_UI_Sounds/FrontEnd/CareerPage/Fortnite_UI_FlagScreenButton_Close_Cue}
    pressedSlateSoundOverride {resourceObject=&Fortnite_UI_FlagScreenButton_Close_Cue}
    selectedHoveredSlateSoundOverride {resourceObject=&Locker_Button_Hover_Default_Cue}
    selectedPressedSlateSoundOverride {resourceObject=&Fortnite_UI_FlagScreenButton_Close_Cue}
    triggeringInputAction {dataTable=&InputActionDataTable, rowName=Close}
   }
   slot {class=HorizontalBoxSlot}
  }
  FoundersNoteText {
   widgetClassPath=&CommonTextBlock
   bIsVariable=true
   parent=VerticalBox_2
   widget=FoundersNoteText
   properties {
    scrollStyle=&TextScrollStyle-Base_C
    text=Note: Founder's Pack banners are awarded after the Tutorial
    font {
     bForceMonospaced=false
     bMaterialIsStencil=false
     fontMaterial=null
     fontObject=&HeadingNow
     letterSpacing=0
     monospacedWidth=1
     size=15.75
     skewAmount=0
     typefaceFontName=74 Regular
     outlineSettings {
      bApplyOutlineToDropShadows=false
      bMiteredCorners=false
      bSeparateFillAlpha=false
      outlineMaterial=null
      outlineSize=0
      outlineColor=<rgba 0 0 0 1>
     }
    }
    shadowOffset=<0 0>
   }
   slot {class=VerticalBoxSlot, properties {verticalAlignment=VAlign_Center}}
  }
  Button_EditIcon {widgetClassPath=&CommonButtonBase, bInherited=true, widget=null}
 ]
}
