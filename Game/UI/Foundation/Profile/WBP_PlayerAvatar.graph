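# Player avatar widget (FortPlayerAvatar) showing a user's avatar, online presence icon and an optional "Sharing" activity badge.
# Events forward social user, item definition or texture to avatar and online-status viewmodels; Construct picks large images from desired size.
# Setter functions drive one dynamic material on AvatarImage for textures, colours, empty state and presence, choosing the status icon via a colour-channel selector.
# PreConstruct applies size, presence-icon layout and visibility flags.
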WidgetBlueprint /Game/UI/Foundation/Profile/WBP_PlayerAvatar
animations [Anim_Hover {
  binding_count=1
  length_seconds=0.2
  object=&self:Anim_Hover
  sequence {
   bindings [AvatarImage {
     id=#1
     possessed_class=Image
     tracks [{
       type=MovieSceneWidgetMaterialTrack
       sections [{type=MovieSceneParameterSection, channels [OverlayOpacity {type=MovieSceneScriptingFloatChannel, keys [[0,0],[12,0.5]]}]}]
      }]
    }]
   display_rate {numerator=60}
  }
 }]
graphs [
 EventGraph {nodes [
   n1 = Event Set Size!() -> (OutputDelegate, then, NewSize:Vector 2D Structure) {is=title K2Node_CustomEvent}
   n2 = Get AvatarImage() -> (AvatarImage:Image Object Reference)
   n3 = Get!(self:Image Object Reference=@n2) -> (
    Brush_TintColor:Slate Color Structure
    Brush_DrawAs:ESlateBrushDrawType Enum
    Brush_Tiling:ESlateBrushTileType Enum
    Brush_Mirroring:ESlateBrushMirrorType Enum
    Brush_ImageSize:Vector 2D Structure
    Brush_Margin:Margin Structure
    Brush_ResourceObject:Object Reference
    Brush_OutlineSettings:Slate Brush Outline Settings Structure
   )
   > SetBrush!(
    execute
    self=@n2
    InBrush_TintColor:Slate Color Structure=@n3
    InBrush_DrawAs:ESlateBrushDrawType Enum=@n3.Brush_DrawAs
    InBrush_Tiling:ESlateBrushTileType Enum=@n3.Brush_Tiling
    InBrush_Mirroring:ESlateBrushMirrorType Enum=@n3.Brush_Mirroring
    InBrush_ImageSize:Vector 2D Structure=@n1.NewSize
    InBrush_Margin:Margin Structure=@n3.Brush_Margin
    InBrush_ResourceObject:Object Reference=@n3.Brush_ResourceObject
    InBrush_OutlineSettings:Slate Brush Outline Settings Structure=@n3.Brush_OutlineSettings
   ) -> (then)
   > SetSize!(execute, self:Spacer Object Reference=Get Spacer_Size() -> (Spacer_Size:Spacer Object Reference), InSize:Vector 2D Structure=@n1.NewSize) -> (then)
   n4 = Event Set Presence Icon Settings!() -> (OutputDelegate, then, Scale:double=~, Offset:Vector 2D Structure) {is=title K2Node_CustomEvent}
   > n5 = GetDynamicMaterial(self=Get AvatarImage())
   > SetScalarParameterValue(self=@n5, ParameterName=PresenceIconScale, Value=@n4.Scale)
   n6 = BreakVector2D(InVec=@n4.Offset)
   > SetScalarParameterValue(self=@n5, ParameterName=PresenceIconOffsetX, Value=@n6)
   > SetScalarParameterValue(self=@n5, ParameterName=PresenceIconOffsetY, Value=@n6.Y)
   Event PreConstruct()
   > n7 = GetDynamicMaterial(self=Get AvatarImage())
   > SetScalarParameterValue(
    self=@n7
    ParameterName=OverrideShowingPresenceIcon
    Value=Select(Option 0:float=0, Option 1:float=1, Index=Get ShowOnlinePresence() -> (ShowOnlinePresence:bool)) -> (ReturnValue:float)
   )
   > SetScalarParameterValue(self=@n7, ParameterName=DisplayPresenceBg, Value=Select(Option 0=0, Option 1=1, Index=Get ShowPresenceBg() -> (ShowPresenceBg:bool)))
   > Event Set Size!(execute, self, NewSize:Vector 2D Structure=Get AvatarSize() -> (AvatarSize:Vector 2D Structure)) -> (then) {is=title K2Node_CallFunction}
   > Event Set Presence Icon Settings!(
    execute
    self
    Scale:double=Get PresenceIconScale() -> (PresenceIconScale:double)
    Offset:Vector 2D Structure=Get PresenceIconOffset() -> (PresenceIconOffset:Vector 2D Structure)
   ) -> (then) {is=title K2Node_CallFunction}
   > SetVisibility(
    self=[
     Get SharingActivity() -> (SharingActivity:Image Object Reference)
     Get WBP_UIKit_Badge_Base() -> (WBP_UIKit_Badge_Base:WBP UIKit Badge Base Object Reference)
    ]
    InVisibility="To Visibility (Boolean)"!(
     bIsVisible:bool=Get Show Sharing Activity Visuals() -> (Show Sharing Activity Visuals:bool)
     TrueVisibility:ESlateVisibility Enum=HitTestInvisible
     FalseVisibility:ESlateVisibility Enum=Collapsed
    ) -> (ReturnValue:ESlateVisibility Enum=Visible)
   )
   n8 = Event SetAvatarFromItemDefinition!() -> (OutputDelegate, then, ItemDefinition:Fort Item Definition Object Reference) {is=title K2Node_Event}
   > SetAvatarFromItemDefinition!(
    execute
    self:Fort Player Avatar Viewmodel Object Reference=Get FortPlayerAvatarViewModel() -> (FortPlayerAvatarViewModel:Fort Player Avatar Viewmodel Object Reference)
    ItemDefinition:Fort Item Definition Object Reference=@n8.ItemDefinition
   ) -> (then)
   Event Construct()
   n9 = Get FortPlayerAvatarViewModel()
   > Branch(Condition=IsValid(Object=@n9))
   n10 = GetDesiredSize!(self:Widget Object Reference=Get AvatarImage()) -> (ReturnValue:Vector 2D Structure)
   > Set bUseLargeImage!(
    execute
    bUseLargeImage:bool="float >= float"(A=GetMin2D!(A:Vector 2D Structure=@n10) -> (ReturnValue:double), B=128)
    self:Fort Player Avatar Viewmodel Object Reference=@n9
   ) -> (then, Output_Get:bool)
   n11 = Event OnSetSocialUser!() -> (OutputDelegate, then, InUser:Fort Social User Object Reference) {is=title K2Node_Event}
   > n12 = Sequence()
   n13 = Get FortPlayerAvatarViewModel()
   > Is Valid(InputObject=@n13)
   > SetSocialUser!(execute, self:Fort Social User Viewmodel Object Reference=@n13, InSocialUser:Fort Social User Object Reference=@n11.InUser) -> (then)
   n14 = Get FortOnlineStatusViewModel() -> (FortOnlineStatusViewModel:Fort Online Status Viewmodel Object Reference)
   Is Valid(exec=@n12.then_1, InputObject=@n14)
   > SetSocialUser(self=@n14, InSocialUser=@n11.InUser)
   n15 = Event SetAvatarTexture!() -> (OutputDelegate, then, InTexture:Texture 2D Object Reference) {is=title K2Node_Event}
   > SetAvatarTexture!(
    execute
    self:Fort Player Avatar Viewmodel Object Reference=Get FortPlayerAvatarViewModel()
    InTexture:Texture 2D Object Reference=@n15.InTexture
   ) -> (then)
  ]}
 SetBackgroundColor {nodes [
   n1 = SetBackgroundColor!() -> (then, BackgroundColor:Color Structure) {is=title K2Node_FunctionEntry}
   n2 = Get AvatarImage()
   > SetVectorParameterValue(
    self=GetDynamicMaterial!(Brush:"Slate Brush Structure (by ref)"=Get Brush!(self:Image Object Reference=@n2) -> (Brush:Slate Brush Structure)) -> (ReturnValue)
    ParameterName=BackgroundRGB
    Value="To LinearColor (Color)"!(InColor:Color Structure=@n1) -> (ReturnValue:Linear Color Structure)
   )
  ]}
 SetEmpty {nodes [
   n1 = SetEmpty!() -> (then, bEmpty:bool=~) {is=title K2Node_FunctionEntry}
   n2 = Get AvatarImage()
   > SetScalarParameterValue(self=GetDynamicMaterial(Brush=Get Brush(self=@n2)), ParameterName=IsEmpty, Value=Select(Option 0=0, Option 1=1, Index=@n1))
  ]}
 SetEmptyBackgroundColor {nodes [
   n1 = SetEmptyBackgroundColor!() -> (then, EmptyBackgroundColor:Color Structure) {is=title K2Node_FunctionEntry}
   n2 = Get AvatarImage()
   > SetVectorParameterValue(self=GetDynamicMaterial(Brush=Get Brush(self=@n2)), ParameterName=EmptyBackgroundRGB, Value="To LinearColor (Color)"(InColor=@n1))
  ]}
 SetEmptyHighlightColor {nodes [
   n1 = SetEmptyHighlightColor!() -> (then, EmptyHighlightColor:Color Structure) {is=title K2Node_FunctionEntry}
   n2 = Get AvatarImage()
   > SetVectorParameterValue(self=GetDynamicMaterial(Brush=Get Brush(self=@n2)), ParameterName=EmptyHighlightRGB, Value="To LinearColor (Color)"(InColor=@n1))
  ]}
 SetHighlightColor {nodes [
   n1 = SetHighlightColor!() -> (then, HighlightColor:Color Structure) {is=title K2Node_FunctionEntry}
   n2 = Get AvatarImage()
   > SetVectorParameterValue(self=GetDynamicMaterial(Brush=Get Brush(self=@n2)), ParameterName=HighlightRGB, Value="To LinearColor (Color)"(InColor=@n1))
  ]}
 SetIconColor {nodes [
   n1 = SetIconColor!() -> (then, IconColor:Color Structure) {is=title K2Node_FunctionEntry}
   n2 = Get AvatarImage()
   > SetVectorParameterValue(self=GetDynamicMaterial(Brush=Get Brush(self=@n2)), ParameterName=Icon Color, Value="To LinearColor (Color)"(InColor=@n1))
  ]}
 SetIconTexture {nodes [
   n1 = SetIconTexture!() -> (then, IconTexture:Texture 2D Object Reference, UseIcon:bool=~) {is=title K2Node_FunctionEntry}
   n2 = IsValid(Object=@n1)
   > SetEmpty!(execute, self, bEmpty:bool=NOT Boolean(A=@n2)) -> (then) {is=title K2Node_CallFunction}
   > Branch(Condition=@n2)
   n3 = GetDynamicMaterial(Brush=Get Brush(self=Get AvatarImage()))
   > SetScalarParameterValue(self=@n3, ParameterName=Use Icon Color Pass, Value=Select(Option 0=0, Option 1=1, Index=@n1.UseIcon))
   > SetTextureParameterValue(self=@n3, ParameterName=IconPreview, Value=@n1)
  ]}
 SetInactiveIncognitoColor {nodes [
   n1 = SetInactiveIncognitoColor!() -> (then, InactiveIncognitoColor:Color Structure) {is=title K2Node_FunctionEntry}
   n2 = Get AvatarImage()
   > SetVectorParameterValue(self=GetDynamicMaterial(Brush=Get Brush(self=@n2)), ParameterName=InactiveIncognitoColor, Value="To LinearColor (Color)"(InColor=@n1))
  ]}
 SetIncognitoColor {nodes [
   n1 = SetIncognitoColor!() -> (then, IncognitoColor:Color Structure) {is=title K2Node_FunctionEntry}
   n2 = Get AvatarImage()
   > SetVectorParameterValue(self=GetDynamicMaterial(Brush=Get Brush(self=@n2)), ParameterName=IncognitoColor, Value="To LinearColor (Color)"(InColor=@n1))
  ]}
 SetOnlineStatus {nodes [
   n1 = SetOnlineStatus!() -> (then, OnlineStatus:EPresenceIndicatorState Enum=Online) {is=title K2Node_FunctionEntry}
   n2 = GetDynamicMaterial(Brush=Get Brush(self=Get AvatarImage()))
   > SetVectorParameterValue(
    self=@n2
    ParameterName=PresenceIconSelector
    Value=Select!(
     Online:Linear Color Structure=Get PresenceSelect_Online() -> (PresenceSelect_Online:Linear Color Structure)
     Away:Linear Color Structure=Get PresenceSelect_Away() -> (PresenceSelect_Away:Linear Color Structure)
     Offline:Linear Color Structure=Get PresenceSelect_Offline() -> (PresenceSelect_Offline:Linear Color Structure)
     Blocked:Linear Color Structure=Get PresenceSelect_Blocked() -> (PresenceSelect_Blocked:Linear Color Structure)
     Index:EPresenceIndicatorState Enum=@n1
    ) -> (ReturnValue:Linear Color Structure)
   )
   n3 = Get PresenceColor_Offline() -> (PresenceColor_Offline:Linear Color Structure)
   > SetVectorParameterValue(
    self=@n2
    ParameterName=PresenceIconColor
    Value=Select(
     Online=Get PresenceColor_Online() -> (PresenceColor_Online:Linear Color Structure)
     Away=Get PresenceColor_Away() -> (PresenceColor_Away:Linear Color Structure)
     Offline=@n3
     Blocked=@n3
     Index=@n1
    )
   )
  ]}
 SetOnlineStatusVisibility {nodes [
   n1 = SetOnlineStatusVisibility!() -> (then, bHideStatus:bool=~) {is=title K2Node_FunctionEntry}
   n2 = Get AvatarImage()
   > SetScalarParameterValue(
    self=GetDynamicMaterial(Brush=Get Brush(self=@n2))
    ParameterName=OverrideShowingPresenceIcon
    Value=Select!(Option 0:float=1, Option 1:float=0, Index:bool=@n1) -> (ReturnValue:float)
   )
  ]}
 SetSharingActivity {nodes [
   n1 = SetSharingActivity!() -> (then, IsLocalPlayerSharingActivity:bool=~) {is=title K2Node_FunctionEntry}
   n2 = Get Show Sharing Activity Visuals()
   > SetVisibility(
    self=[Get WBP_UIKit_Badge_Base(), Get SharingActivity()]
    InVisibility="To Visibility (Boolean)"(bIsVisible=AND Boolean(A=@n1, B=@n2), TrueVisibility=HitTestInvisible, FalseVisibility=Collapsed) -> (ReturnValue=Visible)
   )
  ]}
 SetTexture {nodes [
   n1 = SetTexture!() -> (then, Texture:Texture 2D Object Reference) {is=title K2Node_FunctionEntry}
   n2 = IsValid(Object=@n1)
   > SetEmpty(bEmpty=NOT Boolean(A=@n2))
   > Branch(Condition=@n2)
   n3 = GetDynamicMaterial(Brush=Get Brush(self=Get AvatarImage()))
   > SetTextureParameterValue(self=@n3, ParameterName=SmallPreview, Value=@n1)
   > SetTextureParameterValue(self=@n3, ParameterName=LargePreview)
  ]}
]
variables [
 AvatarMaterial {type=MaterialInstanceDynamic}
 AvatarSize {type=Vector2D, default=<64 64>}
 PresenceColor_Away {type=LinearColor, default=<rgba 0.9647 0.5457 0.1221 1>}
 PresenceColor_Offline {type=LinearColor, default=<rgba 0.4287 0.6654 0.7991 0.302>}
 PresenceColor_Online {type=LinearColor, default=<rgba 0.2159 1 0.1329 1>}
 PresenceIconOffset {type=Vector2D, default=<-0.84 -0.84>}
 PresenceIconScale {type=number, default=0.4}
 PresenceSelect_Away {type=LinearColor, default=<rgba 0 1 0 0>}
 PresenceSelect_Blocked {type=LinearColor, default=<rgba 0 0 0 1>}
 PresenceSelect_Offline {type=LinearColor, default=<rgba 0 0 1 0>}
 PresenceSelect_Online {type=LinearColor, default=<rgba 1 0 0 0>}
 Show Sharing Activity Visuals {type=boolean}
 ShowOnlinePresence {type=boolean, default=true}
 ShowPresenceBg {type=boolean}
]
widgets {
 info {inheritedWidgetCount=0, parentClass=/Script/FortniteUI.FortPlayerAvatar, rootWidgetClass=&Overlay, widgetCount=6}
 widgets [
  Overlay_0 {widgetClassPath=&Overlay, widget=Overlay_0}
  SharingActivity {
   widgetClassPath=&Image
   bIsVariable=true
   parent=Overlay_0
   widget=SharingActivity
   properties {
    visibility=SelfHitTestInvisible
    brush {
     bIsDynamicallyLoaded=false
     drawAs=Image
     imageType=FullColor
     mirroring=NoMirror
     resourceName=None
     resourceObject=/Game/UI/DiscoveryShared/Art/MaterialInstances/MI_UI_SocialLiveActivity
     tiling=NoTile
     imageSize=<32 32>
     margin=<ltrb 0 0 0 0>
     outlineSettings {
      bUseBrushTransparency=false
      roundingType=HalfHeightRadius
      width=0
      color {colorUseRule=UseColor_Specified, specifiedColor=<rgba 0 0 0 0>}
      cornerRadii=<0 0 0 0>
     }
     tintColor {colorUseRule=UseColor_Specified, specifiedColor=<rgba 1 1 1 1>}
     uVRegion {bIsValid=false, max=<0 0>, min=<0 0>}
    }
   }
   slot {class=OverlaySlot, properties {horizontalAlignment=HAlign_Fill, verticalAlignment=VAlign_Fill, padding=<ltrb -25 -25 -25 -25>}}
  }
  AvatarImage {
   widgetClassPath=&Image
   bIsVariable=true
   parent=Overlay_0
   widget=AvatarImage
   properties {brush {
     bIsDynamicallyLoaded=false
     drawAs=Image
     imageType=NoImage
     mirroring=NoMirror
     resourceName=None
     resourceObject=/Game/UI/Foundation/Profile/Materials/MI_UI_PlayerAvatar
     tiling=NoTile
     imageSize=<32 32>
     margin=<ltrb 0 0 0 0>
     outlineSettings {
      bUseBrushTransparency=false
      roundingType=HalfHeightRadius
      width=0
      color {colorUseRule=UseColor_Specified, specifiedColor=<rgba 0 0 0 0>}
      cornerRadii=<0 0 0 0>
     }
     tintColor {colorUseRule=UseColor_Specified, specifiedColor=<rgba 1 1 1 1>}
     uVRegion {bIsValid=false, max=<0 0>, min=<0 0>}
    }}
   slot {class=OverlaySlot, properties {horizontalAlignment=HAlign_Fill, verticalAlignment=VAlign_Fill}}
  }
  FortVisualAttachment_SharingBadge {
   widgetClassPath=&FortVisualAttachment
   bIsVariable=true
   parent=Overlay_0
   widget=FortVisualAttachment_SharingBadge
   properties {contentAnchor=<0.5 -0.5>}
   slot {class=OverlaySlot, properties {horizontalAlignment=HAlign_Center, verticalAlignment=VAlign_Bottom}}
  }
  WBP_UIKit_Badge_Base {
   widgetClassPath=/FortUILibrary/Systems/Badge/WBP_UIKit_Badge_Base
   bIsVariable=true
   parent=FortVisualAttachment_SharingBadge
   widget=WBP_UIKit_Badge_Base
   properties {badgeText=Sharing, badgeType=Text, indicatorConfig=None}
   slot {class=SizeBoxSlot}
  }
  Spacer_Size {
   widgetClassPath=&Spacer
   bIsVariable=true
   parent=Overlay_0
   widget=Spacer_Size
   properties {size=<64 64>}
   slot {class=OverlaySlot, properties {horizontalAlignment=HAlign_Center, verticalAlignment=VAlign_Center}}
  }
 ]
}
